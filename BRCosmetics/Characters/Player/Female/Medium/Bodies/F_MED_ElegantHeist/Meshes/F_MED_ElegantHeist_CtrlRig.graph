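# Control Rig for the F_MED_ElegantHeist outfit that adds corrective and secondary bone motion to the jacket and chain accessory.
# After the shared DeformRig function, upper-arm animation attributes, some curve-remapped, and thigh rotations weight jacket and twist-fix offsets.
# A local DeltaRotation function converts thigh swing/twist into clamped weights.
# Weighted parent and aim constraints position the chain with thigh-driven offsets, and a swimming flag applies further chain offsets.
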ControlRigBlueprint /BRCosmetics/Characters/Player/Female/Medium/Bodies/F_MED_ElegantHeist/Meshes/F_MED_ElegantHeist_CtrlRig
graphs [
 New Function_ContainedGraph {nodes [
   n1 = RigVMFunctionEntryNode()
   RigVMFunction_MathBoolToInteger = &RigVMFunction_MathBoolToInteger(Value=@n1.onlyTwist)
   RigVMFunction_MathQuaternionSwingTwist = &RigVMFunction_MathQuaternionSwingTwist(
    TwistAxis=@n1.twistAxis
    Input=&RigUnit_GetRelativeTransformForItem(
     Child=(Type=Bone,Name="None")
     Parent=(Type=Bone,Name="None")
     bChildInitial=False
     bParentInitial=False
     Child.Name=@n1.child
     Parent.Name=@n1.parent
     Parent.Type=@n1.parentType
    ).RelativeTransform.Rotation
   )
   RigVMFunction_MathQuaternionToEuler = &RigVMFunction_MathQuaternionToEuler(
    Value=RigVMDispatchNode(
     Values=((),())
     Index=@RigVMFunction_MathBoolToInteger.Result
     Values.0=@RigVMFunction_MathQuaternionSwingTwist.Swing
     Values.1=@RigVMFunction_MathQuaternionSwingTwist.Twist
    ) {template="DISPATCH_RigVMDispatch_SelectInt32(in Index,in Values,out Result)"}.Result
    RotationOrder=@n1.Rotation Order
   )
   > DISPATCH_RigVMDispatch_Print = RigVMDispatchNode(ScreenColor=(R=1,G=1,B=1,A=1), ScreenDuration=1.5, Enabled=@n1.debugRotation, Value=@RigVMFunction_MathQuaternionToEuler.Result) {template="DISPATCH_RigVMDispatch_Print(in Prefix,in Value,in Enabled,in ScreenDuration,in ScreenColor)"}
   Clamp = &RigVMFunction_MathIntClamp(Value=@n1.primAxis)
   DISPATCH_RigVMDispatch_SelectInt32 = RigVMDispatchNode(
    Values=(0,0,0)
    Index=@Clamp.Result
    Values.0=@RigVMFunction_MathQuaternionToEuler.Result.X
    Values.1=@RigVMFunction_MathQuaternionToEuler.Result.Y
    Values.2=@RigVMFunction_MathQuaternionToEuler.Result.Z
   ) {template="DISPATCH_RigVMDispatch_SelectInt32(in Index,in Values,out Result)"}
   Remap = &RigVMFunction_MathDoubleRemap(
    TargetMaximum=1
    TargetMinimum=0
    bClamp=False
    Value=@DISPATCH_RigVMDispatch_SelectInt32.Result
    SourceMaximum=@n1.max
    SourceMinimum=@n1.min
   )
   Multiply = &RigVMFunction_MathDoubleMul(A=&RigVMFunction_MathDoubleClamp(Value=@Remap.Result).Result, B=@n1.premult)
   > RigVMFunctionReturnNode(outputWeight=@Multiply.Result)
  ]}
 RigVMFunctionLibrary {nodes [DeltaRotation = RigVMCollapseNode(ExecuteContext=(), Rotation Order=ZYX, premult=1, twistAxis=(X=1,Y=0,Z=0))]}
 RigVMFunction_Sequence_1_ContainedGraph {nodes [
   n1 = RigVMFunctionEntryNode()
   RigVMFunction_Sequence = &RigVMFunction_Sequence(ExecuteContext=@n1.ExecuteContext)
   n2 = &RigVMFunction_Sequence(ExecuteContext=@RigVMFunction_Sequence.B)
   n3 = &RigVMFunction_Sequence(ExecuteContext=@n2.B)
   RigVMFunctionReturnNode(A=@RigVMFunction_Sequence.A, B=@n2.A, C=@n3.A, D=@n3.B)
  ]}
 RigVMModel {nodes [
   RigUnit_BeginExecution = &RigUnit_BeginExecution()
   RigVMFunction_Sequence = RigVMAggregateNode(ExecuteContext=@RigUnit_BeginExecution.ExecutePin)
   FN_DeformRig_FNC = RigVMFunctionReferenceNode(
    DeformRig_ Curve="(deformRig_l_upperarm_up_alpha_72_#1=true,deformRig_r_upperarm_up_alpha_73_#2=true,deformRig_l_upperarm_fr_alpha_74_#3=true,deformRig_r_upperarm_fr_alpha_75_CB51B80C4 ...(truncated)"
    DeformRig_Bool="(upperarm_1_#4=True,elbow_3_#5=True,wrist_7_#6=True,groin_8_#7=True,glute_17_#8=True,knee_9_40C408324CA261 ...(truncated)"
    DeformRig_Debug=(upperarm_debug_18_#4=false,wrist_debug_20_#6=False,lat_debug_26_#9=False,groinGlute_debug_27_#7=False)
    DeformRig_Value="(upperarm_up_18_#10=(Rotation=(X=0,Y=0,Z=0,W=1),Translation=(X=0,Y=0,Z=8),Scale3D=(X=0.7,Y=1,Z=1)),upperarm_fwd_fr_19_#11=(Rotation=(X=0,Y=0,Z=0,W=1),Translation=(X=0,Y=10,Z=0),Scale3D=(X=0.5,Y=1,Z= ...(truncated)"
    ExecuteContext=@RigVMFunction_Sequence.B
   )
   Get_Animation_Attribute = RigVMDispatchNode(BoneName=Root, Default=0, Name=deformRig_l_upperarm_up_alpha) {template="DISPATCH_RigDispatch_GetAnimAttribute(in Name,in BoneName,in Default,out Value,out Found)"}
   > &RigUnit_ModifyTransforms(
    ItemToModify="((Item=(Type=Bone,Name=\"dyn_jacket_a_01_l\"),Transform=(Rotation=(X=-0.1403,Y=0.3809,Z=0.3505,W=0.844),Translation=(X=0,Y=0,Z=0),Scale3D=(X=1,Y=1,Z=1))),(Item=(Type=Bone,Name=\"dyn_jacket_b_01_l\"),Transform=(Rotation=(X=0.3309,Y=-0.3627,Z=0.2428,W=0.8367),Tr ...(truncated)"
    WeightMinimum=0
    Weight=@Get_Animation_Attribute.Value
   )
   Get_Animation_Attribute_2 = RigVMDispatchNode(BoneName=Root, Default=0, Name=deformRig_l_upperarm_fr_alpha) {template="DISPATCH_RigDispatch_GetAnimAttribute(in Name,in BoneName,in Default,out Value,out Found)"}
   n1 = &RigVMFunction_AnimEvalRichCurve(
    Curve=(EditorCurveData=(Keys=((Time=-0.0056,Value=-0.009),(InterpMode=RCIM_Cubic,TangentMode=RCTM_User,Time=1,Value=1,ArriveTangent=3.24,LeaveTangent=3.24))))
    SourceMaximum=0.3
    SourceMinimum=0
    TargetMinimum=0
    Value=@Get_Animation_Attribute_2.Value
   )
   > &RigUnit_ModifyTransforms(
    ItemToModify="((Item=(Type=Bone,Name=\"dyn_jacket_a_01_l\"),Transform=(Rotation=(X=-0.1331,Y=0.1956,Z=-0.0969,W=0.9668),Translation=(X=-9.9749,Y=2.6407,Z=3.8853),Scale3D=(X=1,Y=1,Z=1))),(Item=(Type=Bone,Name=\"dyn_jacket_b_01_l\"),Transform=(Rotation=(X=0.059,Y=0.1607,Z=-0. ...(truncated)"
    WeightMinimum=0
    Weight=@n1.Result
   )
   > &RigUnit_ModifyTransforms(
    ItemToModify=((Item=(Type=Bone,Name="upperarm_twist_01_bffix_l"),Transform=(Rotation=(X=-0.4428,Y=-0.4384,Z=-0.5684,W=0.5373),Translation=(X=11.6593,Y=-15.8821,Z=6.2264),Scale3D=(X=1,Y=1,Z=1))))
    WeightMinimum=0
    Weight=@Get_Animation_Attribute_2.Value
   )
   > DeltaRotation_1 = RigVMFunctionReferenceNode(
    Rotation Order=ZYX
    child=thigh_l
    debugRotation=false
    max=-100
    min=-80
    parent=thigh_l_null
    parentType=Null
    premult=1
    primAxis=2
    twistAxis=(X=1,Y=0,Z=0)
   )
   > &RigUnit_ModifyTransforms(
    ItemToModify="((Item=(Type=Bone,Name=\"dyn_jacket_d_01_l\"),Transform=(Rotation=(X=0.1823,Y=-0.5794,Z=0.0595,W=0.7921),Translation=(X=0,Y=0,Z=0),Scale3D=(X=1,Y=1,Z=1))),(Item=(Type=Bone,Name=\"dyn_jacket_d_02_l\"),Transform=(Rotation=(X=0,Y=-0.1736,Z=0,W=0.9848),Translation ...(truncated)"
    WeightMinimum=0
    Weight=@DeltaRotation_1.outputWeight
   )
   Get_Animation_Attribute_3 = RigVMDispatchNode(BoneName=Root, Default=0, Name=deformRig_r_upperarm_up_alpha) {template="DISPATCH_RigDispatch_GetAnimAttribute(in Name,in BoneName,in Default,out Value,out Found)"}
   > &RigUnit_ModifyTransforms(
    ItemToModify="((Item=(Type=Bone,Name=\"base_jacket_a_01_r\"),Transform=(Rotation=(X=0.2974,Y=-0.1031,Z=-0.1094,W=0.9428),Translation=(X=0,Y=0,Z=0),Scale3D=(X=1,Y=1,Z=1))),(Item=(Type=Bone,Name=\"dyn_jacket_b_01_r\"),Transform=(Rotation=(X=-0.0661,Y=-0.4507,Z=0.2173,W=0.8633 ...(truncated)"
    WeightMinimum=0
    Weight=@Get_Animation_Attribute_3.Value
   )
   Get_Animation_Attribute_1 = RigVMDispatchNode(BoneName=Root, Default=0, Name=deformRig_r_upperarm_fr_alpha) {template="DISPATCH_RigDispatch_GetAnimAttribute(in Name,in BoneName,in Default,out Value,out Found)"}
   > &RigUnit_ModifyTransforms(
    ItemToModify=((Item=(Type=Bone,Name="upperarm_twist_01_bffix_r"),Transform=(Rotation=(X=0.1448,Y=0.0654,Z=0.3916,W=0.9063),Translation=(X=-8.187,Y=12.458,Z=-0.8357),Scale3D=(X=1,Y=1,Z=1))))
    WeightMinimum=0
    Weight=@Get_Animation_Attribute_1.Value
   )
   n2 = &RigVMFunction_AnimEvalRichCurve(
    Curve=(EditorCurveData=(Keys=((),(InterpMode=RCIM_Cubic,TangentMode=RCTM_User,Time=1,Value=1,ArriveTangent=3.24,LeaveTangent=3.24))))
    SourceMaximum=0.36
    SourceMinimum=0
    TargetMinimum=0
    Value=@Get_Animation_Attribute_1.Value
   )
   > &RigUnit_ModifyTransforms(
    ItemToModify="((Item=(Type=Bone,Name=\"base_jacket_a_01_r\"),Transform=(Rotation=(X=-0.0315,Y=-0.0024,Z=-0.4214,W=0.9063),Translation=(X=0,Y=0,Z=0),Scale3D=(X=1,Y=1,Z=1))),(Item=(Type=Bone,Name=\"dyn_jacket_b_01_r\"),Transform=(Rotation=(X=0.0141,Y=0.1547,Z=-0.3036,W=0.94), ...(truncated)"
    WeightMinimum=0
    Weight=@n2.Result
   )
   > DeltaRotation_2 = RigVMFunctionReferenceNode(
    Rotation Order=ZYX
    child=thigh_r
    debugRotation=false
    max=-100
    min=-70
    parent=thigh_r_null
    parentType=Null
    premult=1
    primAxis=2
    twistAxis=(X=1,Y=0,Z=0)
   )
   > &RigUnit_ModifyTransforms(
    ItemToModify="((Item=(Type=Bone,Name=\"dyn_jacket_d_01_r\"),Transform=(Rotation=(X=0,Y=0.2427,Z=0,W=0.9701),Translation=(X=-0.576,Y=0,Z=-0.3794),Scale3D=(X=1,Y=1,Z=1))),(Item=(Type=Bone,Name=\"None\"),Transform=(Rotation=(X=0,Y=0,Z=0,W=1),Translation=(X=0,Y=0,Z=0),Scale3D=( ...(truncated)"
    WeightMinimum=0
    Weight=@DeltaRotation_2.outputWeight
   )
   > &RigUnit_ModifyTransforms(
    ItemToModify=((Item=(Type=Bone,Name="helper_spine_03"),Transform=(Rotation=(X=0,Y=0,Z=0,W=1),Translation=(X=0,Y=0,Z=0),Scale3D=(X=0.8,Y=0.8,Z=0.8))))
    Weight=1
    WeightMinimum=0
   )
   &RigUnit_ParentConstraint(
    AdvancedSettings=(RotationOrderForFilter=XZY)
    Child=(Type=Bone,Name="aim_chain_front_chain")
    Parents=((Item=(Type=Bone,Name="pelvis"),Weight=0.0002),(Item=(Type=Bone,Name="deform_glute_r"),Weight=0.012),(Item=(Type=Bone,Name="deform_groin_r"),Weight=0.9878))
    Weight=1
    ExecutePin=@RigVMFunction_Sequence.C
   )
   > &RigUnit_ParentConstraint(
    AdvancedSettings=(RotationOrderForFilter=XZY)
    Child=(Type=Bone,Name="aim_chain_back_chain")
    Parents=((Item=(Type=Bone,Name="spine_01"),Weight=0.0008),(Item=(Type=Bone,Name="deform_groin_r"),Weight=0.0127),(Item=(Type=Bone,Name="deform_glute_r"),Weight=0.9865))
    Weight=1
   )
   > &RigUnit_ParentConstraint(
    AdvancedSettings=(RotationOrderForFilter=XZY)
    Child=(Type=Bone,Name="base_chain_bck")
    Parents="((Item=(Type=Bone,Name=\"spine_02\"),Weight=0.0011),(Item=(Type=Bone,Name=\"spine_01\"),Weight=0.0022),(Item=(Type=Bone,Name=\"thigh_twist_01_l\"),Weight=0.0032),(Item=(Type=Bone,Name=\"thigh_r\"),Weight=0.0044),(Item=(Type=Bone,Name=\"thigh_twist_01_r\"),Weight=0.0 ...(truncated)"
    Weight=1
   )
   > &RigUnit_ParentConstraint(
    AdvancedSettings=(RotationOrderForFilter=XZY)
    Child=(Type=Bone,Name="base_chain_fr")
    Parents="((Item=(Type=Bone,Name=\"thigh_r\"),Weight=0.0049),(Item=(Type=Bone,Name=\"thigh_l\"),Weight=0.0049),(Item=(Type=Bone,Name=\"spine_01\"),Weight=0.0057),(Item=(Type=Bone,Name=\"thigh_twist_01_l\"),Weight=0.019),(Item=(Type=Bone,Name=\"thigh_twist_01_r\"),Weight=0.019 ...(truncated)"
    Weight=1
   )
   > &RigUnit_ParentConstraint(
    AdvancedSettings=(RotationOrderForFilter=XZY)
    Child=(Type=Bone,Name="aim_chain_fr_01")
    Parents="((Item=(Type=Bone,Name=\"deform_glute_r\"),Weight=0.0066),(Item=(Type=Bone,Name=\"thigh_l\"),Weight=0.0076),(Item=(Type=Bone,Name=\"deform_groin_l\"),Weight=0.0101),(Item=(Type=Bone,Name=\"thigh_twist_01_l\"),Weight=0.0292),(Item=(Type=Bone,Name=\"thigh_r\"),Weight= ...(truncated)"
    Weight=1
   )
   > &RigUnit_ParentConstraint(
    AdvancedSettings=(RotationOrderForFilter=XZY)
    Child=(Type=Bone,Name="aim_chain_acc_cn_02")
    Parents=((Item=(Type=Bone,Name="dyn_chain_acc_cn_02"),Weight=1.0))
    Weight=1
   )
   > &RigUnit_ParentConstraint(
    AdvancedSettings=(RotationOrderForFilter=XZY)
    Child=(Type=Bone,Name="aim_chain_bck_01")
    Parents="((Item=(Type=Bone,Name=\"thigh_twist_01_l\"),Weight=0.0017),(Item=(Type=Bone,Name=\"thigh_r\"),Weight=0.0226),(Item=(Type=Bone,Name=\"thigh_twist_01_r\"),Weight=0.1043),(Item=(Type=Bone,Name=\"deform_glute_r\"),Weight=0.4123),(Item=(Type=Bone,Name=\"pelvis\"),Weight ...(truncated)"
    Weight=1
   )
   > DeltaRotation_0 = RigVMFunctionReferenceNode(
    Rotation Order=ZYX
    child=thigh_r
    debugRotation=false
    max=-100
    parent=thigh_r_null
    parentType=Null
    premult=1
    primAxis=2
    twistAxis=(X=1,Y=0,Z=0)
   )
   > &RigUnit_ModifyTransforms(
    ItemToModify="((Item=(Type=Bone,Name=\"aim_chain_acc_cn_02\"),Transform=(Rotation=(X=0,Y=0,Z=0,W=1),Translation=(X=-14.8616,Y=-13.794,Z=7.3244),Scale3D=(X=1,Y=1,Z=1))),(Item=(Type=Bone,Name=\"helper_thigh_r\"),Transform=(Rotation=(X=0,Y=0,Z=-0.0391,W=0.9992),Translation=(X= ...(truncated)"
    WeightMinimum=0
    Weight=@DeltaRotation_0.outputWeight
   )
   > &RigUnit_AimConstraintLocalSpaceOffset(
    Child=(Type=Bone,Name="dyn_chain_fr_00")
    Filter=(bX=True,bY=True,bZ=True)
    Parents=((Item=(Type=Bone,Name="aim_chain_fr_01"),Weight=1))
    Weight=1
    WorldUp=(Target=(X=0,Y=0,Z=1),Kind=Direction,Space=(Type=None,Name="None"))
   )
   > &RigUnit_AimConstraintLocalSpaceOffset(
    Child=(Type=Bone,Name="dyn_chain_bck_00")
    Filter=(bX=True,bY=True,bZ=True)
    Parents=((Item=(Type=Bone,Name="aim_chain_bck_01"),Weight=1))
    Weight=1
    WorldUp=(Target=(X=0,Y=0,Z=1),Kind=Direction,Space=(Type=None,Name="None"))
   )
   > &RigUnit_AimConstraintLocalSpaceOffset(
    Child=(Type=Bone,Name="dyn_chain_fr_01")
    Filter=(bX=True,bY=True,bZ=True)
    Parents=((Item=(Type=Bone,Name="aim_chain_acc_cn_02_2"),Weight=1))
    Weight=1
    WorldUp=(Target=(X=0,Y=0,Z=1),Kind=Direction,Space=(Type=None,Name="None"))
   )
   > &RigUnit_AimConstraintLocalSpaceOffset(
    Child=(Type=Bone,Name="dyn_chain_bck_01")
    Filter=(bX=True,bY=True,bZ=True)
    Parents=((Item=(Type=Bone,Name="aim_chain_acc_cn_02_2"),Weight=1))
    UpAxis=(X=0,Y=1,Z=0)
    Weight=1
    WorldUp=(Target=(X=0,Y=0,Z=1),Kind=Direction,Space=(Type=None,Name="None"))
   )
   > &RigUnit_ModifyTransforms(
    ItemToModify="((Item=(Type=Bone,Name=\"dyn_chain_fr_01\"),Transform=(Rotation=(X=0,Y=0,Z=0,W=1),Translation=(X=-0.8741,Y=9.4366,Z=-1.5509),Scale3D=(X=1,Y=1,Z=1))),(Item=(Type=Bone,Name=\"dyn_chain_fr_02\"),Transform=(Rotation=(X=0,Y=0,Z=0,W=1),Translation=(X=-4.8209,Y=7.166 ...(truncated)"
    WeightMinimum=0
    Weight=@DeltaRotation_0.outputWeight
   )
   RigVMFunction_ControlFlowBranch = &RigVMFunction_ControlFlowBranch(ExecuteContext=@RigVMFunction_Sequence.D, Condition=RigVMVariableNode() {variable=swimming}.Value)
   &RigUnit_ModifyTransforms(
    ItemToModify="((Item=(Type=Bone,Name=\"dyn_chain_bck_00\"),Transform=(Rotation=(X=0,Y=0,Z=0,W=-1),Translation=(X=0,Y=5.0352,Z=0),Scale3D=(X=1,Y=1,Z=1))),(Item=(Type=Bone,Name=\"dyn_chain_bck_01\"),Transform=(Rotation=(X=0,Y=0,Z=0,W=-1),Translation=(X=0,Y=0,Z=3.5126),Scale3D ...(truncated)"
    Weight=1
    WeightMinimum=0
    ExecutePin=@RigVMFunction_ControlFlowBranch.True
   )
   RigUnit_GetTransform = &RigUnit_GetTransform(Item=(Type=Control,Name="upperarm_l_ctrl"), bInitial=false)
   RigUnit_SetTransform = &RigUnit_SetTransform(ExecutePin=(), Item=(Type=Bone,Name="upperarm_l"), Weight=1, bInitial=false, Value=@RigUnit_GetTransform.Transform)
   > &RigUnit_SetTransform(
    Item=(Type=Bone,Name="upperarm_r")
    Weight=1
    bInitial=False
    Value=&RigUnit_GetTransform(Item=(Type=Control,Name="upperarm_r_ctrl"), bInitial=False).Transform
   )
   RigVMRerouteNode(Value=@RigVMFunction_Sequence.A)
  ]}
]
hierarchy [
 root {loc=<0 0 0>, rot=<0 0 0 1>}
 pelvis {parent=root, loc=<0 -0.5301 98.0571>, rot=<0.0127 0.707 -0.0127 -0.707>}
 spine_01 {parent=pelvis, loc=<6.1208 -0.9328 0>, rot=<0 0 0.034 -0.9994>}
 spine_02 {parent=spine_01, loc=<6.5997 -0.2095 0>, rot=<0 0 0.0238 -0.9997>}
 spine_03 {parent=spine_02, loc=<6.4914 -0.1237 0>, rot=<0 0 0.0149 -0.9999>}
 spine_04 {parent=spine_03, loc=<6.9598 0.0794 0>, rot=<0 0 -0.0132 -0.9999>}
 spine_05 {parent=spine_04, loc=<6.9987 0 0>, rot=<0 0 -0.0015 -1>}
 attach_cape {parent=spine_05, loc=<9.016 -11.7106 -0.0013>, rot=<0.0616 -0.7044 -0.0616 -0.7044>}
 neck_01 {parent=spine_05, loc=<11.3138 0.1529 0>, rot=<0 0 -0.2003 -0.9797>}
 neck_02 {parent=neck_01, loc=<4.5627 0 0>, rot=<0 0 0.0504 -0.9987>}
 head {parent=neck_02, loc=<5.3914 0 0>, rot=<0 0 0.115 -0.9934>}
 base_hair_back {parent=head, loc=<16.6486 -6.4972 -0.1508>, rot=<0.4335 -0.9011 0 0>}
 base_hair_back_l {parent=base_hair_back, loc=<7.8286 0.553 6.7091>, rot=<0 0 -0.43 -0.9028>}
 dyn_hair_back_l_01 {parent=base_hair_back_l, loc=<0 0 0>, rot=<0 0 0 1>}
 dyn_hair_back_l_02 {parent=dyn_hair_back_l_01, loc=<13.4489 0.0498 0>, rot=<0 0 0.2094 -0.9778>}
 dyn_hair_back_l_03 {parent=dyn_hair_back_l_02, loc=<11.1308 -0.0859 0>, rot=<0 0 -0.1373 -0.9905>}
 dyn_hair_back_l_04 {parent=dyn_hair_back_l_03, loc=<10.0891 0.0023 0>, rot=<0 0 0 -1>}
 dyn_hair_back_l_02b {parent=dyn_hair_back_l_02, loc=<2.8835 1.8226 5.8218>, rot=<0 0 0.0087 -1>}
 dyn_hair_back_l_03b {parent=dyn_hair_back_l_02b, loc=<14.0342 0.0116 0.0146>, rot=<0 0 0 -1>}
 base_hair_back_r {parent=base_hair_back, loc=<8.5201 0.5701 -8.0197>, rot=<0 0 -0.4062 -0.9138>}
 dyn_hair_back_r_01 {parent=base_hair_back_r, loc=<0 0 0>, rot=<0 0 0 1>}
 dyn_hair_back_r_02 {parent=dyn_hair_back_r_01, loc=<11.8864 -0.0262 0>, rot=<0 0 0.1254 -0.9921>}
 dyn_hair_back_r_03 {parent=dyn_hair_back_r_02, loc=<12.085 -0.0239 0>, rot=<0 0 -0.1194 -0.9929>}
 dyn_hair_back_r_04 {parent=dyn_hair_back_r_03, loc=<9.2532 0.0704 0>, rot=<0 0 0 -1>}
 dyn_hair_back_r_02b {parent=dyn_hair_back_r_02, loc=<4.9361 3.7632 -7.621>, rot=<0 0 0.0677 -0.9977>}
 dyn_hair_back_r_03b {parent=dyn_hair_back_r_02b, loc=<8.4203 -0.047 0.0249>, rot=<0 0 0 -1>}
 base_hair_back_m {parent=base_hair_back, loc=<8.4486 -0.0349 -0.3581>, rot=<0 0 -0.4182 -0.9084>}
 dyn_hair_back_m_01 {parent=base_hair_back_m, loc=<0 0 0>, rot=<0 0 0 1>}
 dyn_hair_back_m_02 {parent=dyn_hair_back_m_01, loc=<13.1016 0.0064 0.0003>, rot=<0 0 0.1863 -0.9825>}
 dyn_hair_back_m_03 {parent=dyn_hair_back_m_02, loc=<11.1018 -0.0171 0>, rot=<0 0 -0.1906 -0.9817>}
 dyn_hair_back_m_04 {parent=dyn_hair_back_m_03, loc=<9.8854 0.0621 0>, rot=<0 0 0 -1>}
 base_earing_l {parent=head, loc=<-0.316 0.331 -6.002>, rot=<0.004 -0.9497 0.3132 -0.0013>}
 dyn_earing_l {parent=base_earing_l, loc=<0 0 0>, rot=<0 0 0 1>}
 base_earing_r {parent=head, loc=<-0.316 0.3315 6.002>, rot=<0.0039 -0.9483 -0.3173 0.0013>}
 dyn_earing_r {parent=base_earing_r, loc=<0 0 0>, rot=<0 0 0 1>}
 FACIAL_C_FacialRoot {parent=head, loc=<1.826 0.6057 0>, rot=<0.5021 -0.4979 -0.5021 -0.4979>}
 FACIAL_L_CheekLower {parent=FACIAL_C_FacialRoot, loc=<4.3 3.9208 4.5346>, rot=<0.0345 -0.427 0.0577 -0.9018>}
 FACIAL_R_CheekLower {parent=FACIAL_C_FacialRoot, loc=<-4.3079 3.9549 4.5149>, rot=<0.0365 0.4272 -0.0612 -0.9013>}
 FACIAL_C_UnderChin {parent=FACIAL_C_FacialRoot, loc=<-0.0068 8.5365 2.9634>, rot=<0.4991 -0.0001 -0.0001 -0.8666>}
 FACIAL_L_UnderChin {parent=FACIAL_C_FacialRoot, loc=<2.8095 7.1968 1.8843>, rot=<0.401 -0.3539 0.0837 -0.8408>}
 FACIAL_R_UnderChin {parent=FACIAL_C_FacialRoot, loc=<-2.8499 7.1717 1.9067>, rot=<0.4012 0.354 -0.0832 -0.8407>}
 FACIAL_C_Jaw {parent=FACIAL_C_FacialRoot, loc=<0 2.0037 1.1413>, rot=<0.342 0 0 -0.9397>}
 FACIAL_C_Jawline {parent=FACIAL_C_Jaw, loc=<0.0046 2.6983 6.984>, rot=<0.2557 0.0004 -0.0002 -0.9667>}
 FACIAL_L_ChinSide {parent=FACIAL_C_Jaw, loc=<2.8025 0.6534 6.7072>, rot=<-0.2207 -0.3739 0.05 -0.8994>}
 FACIAL_R_ChinSide {parent=FACIAL_C_Jaw, loc=<-2.8423 0.6386 6.6488>, rot=<-0.2129 0.3733 -0.0477 -0.9017>}
 FACIAL_C_Chin {parent=FACIAL_C_Jaw, loc=<-0.0007 1.6541 6.9807>, rot=<-0.2588 0 0 -0.9659>}
 FACIAL_L_Jawline {parent=FACIAL_C_FacialRoot, loc=<4.1213 5.5029 3.1428>, rot=<0.2059 0.1576 -0.0552 -0.9642>}
 FACIAL_R_Jawline {parent=FACIAL_C_FacialRoot, loc=<-4.1191 5.4981 3.1536>, rot=<0.2073 -0.1577 0.055 -0.9639>}
 FACIAL_C_Forehead {parent=FACIAL_C_FacialRoot, loc=<0.0001 -2.428 7.4971>, rot=<-0.0464 0 0 -0.9989>}
 FACIAL_L_ForeheadIn {parent=FACIAL_C_FacialRoot, loc=<1.723 -2.9403 7.3899>, rot=<-0.0611 -0.0666 0.0062 -0.9959>}
 FACIAL_R_ForeheadIn {parent=FACIAL_C_FacialRoot, loc=<-1.6965 -2.9979 7.3731>, rot=<-0.061 0.0669 -0.0059 -0.9959>}
 FACIAL_L_ForeheadMid {parent=FACIAL_C_FacialRoot, loc=<3.1191 -3.166 6.9051>, rot=<-0.0713 -0.1748 -0.0281 -0.9816>}
 FACIAL_R_ForeheadMid {parent=FACIAL_C_FacialRoot, loc=<-3.0818 -3.2172 6.9051>, rot=<-0.0714 0.1749 0.0268 -0.9816>}
 FACIAL_L_ForeheadOut {parent=FACIAL_C_FacialRoot, loc=<4.5575 -3.0271 6.2029>, rot=<-0.1242 -0.2561 -0.0842 -0.9549>}
 FACIAL_R_ForeheadOut {parent=FACIAL_C_FacialRoot, loc=<-4.5064 -3.102 6.1995>, rot=<-0.1233 0.2563 0.0829 -0.9551>}
 FACIAL_L_EyesackUpper {parent=FACIAL_C_FacialRoot, loc=<3.3129 -1.9229 7.7458>, rot=<0.0564 -0.083 -0.0559 -0.9934>}
 FACIAL_R_EyesackUpper {parent=FACIAL_C_FacialRoot, loc=<-3.2892 -1.9678 7.7933>, rot=<0.0491 0.0854 0.0627 -0.9932>}
 FACIAL_L_EyelidUpperFurrow {parent=FACIAL_C_FacialRoot, loc=<2.9892 -2.2085 7.7712>, rot=<-0.0301 -0.1509 -0.0796 -0.9849>}
 FACIAL_R_EyelidUpperFurrow {parent=FACIAL_C_FacialRoot, loc=<-3.1585 -2.2823 7.7526>, rot=<-0.0305 0.1506 0.0786 -0.985>}
 FACIAL_L_EyelidUpperA {parent=FACIAL_C_FacialRoot, loc=<2.9017 -1.4759 5.7628>, rot=<0.2941 0.0704 -0.0739 -0.9503>}
 FACIAL_R_EyelidUpperA {parent=FACIAL_C_FacialRoot, loc=<-2.9262 -1.5609 5.6195>, rot=<-0.1478 0 0.0001 -0.989>}
 FACIAL_L_Eye {parent=FACIAL_C_FacialRoot, loc=<2.9262 -1.5616 5.6195>, rot=<0 0 0.0001 -1>}
 FACIAL_R_Eye {parent=FACIAL_C_FacialRoot, loc=<-2.9046 -1.5642 5.626>, rot=<0 -0.153 0 -0.9882>}
 FACIAL_L_EyelidLowerA {parent=FACIAL_C_FacialRoot, loc=<2.9262 -1.5616 5.6195>, rot=<0.284 0 0.0001 -0.9588>}
 FACIAL_R_EyelidLowerA {parent=FACIAL_C_FacialRoot, loc=<-2.926 -1.5642 5.626>, rot=<0.2804 -0.0061 -0.008 -0.9598>}
 FACIAL_L_EyeCornerInner {parent=FACIAL_C_FacialRoot, loc=<1.7897 -0.7599 6.9141>, rot=<0 0 0.0001 -1>}
 FACIAL_R_EyeCornerInner {parent=FACIAL_C_FacialRoot, loc=<-1.7895 -0.7595 6.9141>, rot=<0 0 0.0001 -1>}
 FACIAL_L_EyeCornerOuter {parent=FACIAL_C_FacialRoot, loc=<4.2749 -1.1069 7.0297>, rot=<0.0032 -0.3744 0.0043 -0.9272>}
 FACIAL_R_EyeCornerOuter {parent=FACIAL_C_FacialRoot, loc=<-4.5776 -1.4332 6.5295>, rot=<0 0.3827 0 -0.9239>}
 FACIAL_L_EyesackLower {parent=FACIAL_C_FacialRoot, loc=<2.9065 0.75 6.0267>, rot=<-0.0223 -0.1398 -0.0222 -0.9897>}
 FACIAL_R_EyesackLower {parent=FACIAL_C_FacialRoot, loc=<-2.8796 0.7359 6.0778>, rot=<-0.0217 0.135 0.0231 -0.9903>}
 FACIAL_L_CheekInner {parent=FACIAL_C_FacialRoot, loc=<2.1455 0.9629 6.5255>, rot=<-0.1233 -0.088 -0.1971 -0.9686>}
 FACIAL_R_CheekInner {parent=FACIAL_C_FacialRoot, loc=<-2.1405 0.9497 6.5376>, rot=<-0.1243 0.09 0.1975 -0.9682>}
 FACIAL_L_CheekOuter {parent=FACIAL_C_FacialRoot, loc=<4.1104 0.8015 5.1585>, rot=<0.0551 -0.377 0.1004 -0.9191>}
 FACIAL_R_CheekOuter {parent=FACIAL_C_FacialRoot, loc=<-4.1159 0.9658 5.1361>, rot=<0.0565 0.3762 -0.1 -0.9194>}
 FACIAL_L_NasolabialBulge {parent=FACIAL_C_FacialRoot, loc=<3.6224 4.1946 5.2371>, rot=<0.0578 -0.3774 0.0239 -0.9239>}
 FACIAL_R_NasolabialBulge {parent=FACIAL_C_FacialRoot, loc=<-3.6531 4.3167 5.1723>, rot=<0.0597 0.3706 -0.0356 -0.9262>}
 FACIAL_L_NasolabialFurrow {parent=FACIAL_C_FacialRoot, loc=<2.8388 2.8309 7.1535>, rot=<-0.0333 -0.1259 -0.3637 -0.9224>}
 FACIAL_R_NasolabialFurrow {parent=FACIAL_C_FacialRoot, loc=<-2.812 3.0889 7.0311>, rot=<-0.0329 0.1267 0.3635 -0.9223>}
 FACIAL_C_Nose {parent=FACIAL_C_FacialRoot, loc=<0 0.7973 7.7667>, rot=<-0.1305 0 0.0003 -0.9914>}
 FACIAL_C_NoseLower {parent=FACIAL_C_Nose, loc=<-0.0003 2.423 0.0165>, rot=<0.1305 0 0 -0.9914>}
 FACIAL_L_Nostril {parent=FACIAL_C_Nose, loc=<0.9965 2.0693 -0.1059>, rot=<0.0173 0.1305 0.0023 -0.9913>}
 FACIAL_R_Nostril {parent=FACIAL_C_Nose, loc=<-0.997 2.0692 -0.1059>, rot=<0.0173 -0.1305 -0.0023 -0.9913>}
 FACIAL_C_NoseTip {parent=FACIAL_C_Nose, loc=<-0.0002 1.7047 0.7204>, rot=<0.0872 0 0 -0.9962>}
 FACIAL_C_MouthUpper {parent=FACIAL_C_FacialRoot, loc=<0 4.4708 6.0634>, rot=<0 0 0 -1>}
 FACIAL_C_LipUpper {parent=FACIAL_C_MouthUpper, loc=<0 -0.2664 1.7497>, rot=<-0.0872 0 0 -0.9962>}
 FACIAL_L_LipUpper {parent=FACIAL_C_MouthUpper, loc=<1.2118 -0.1743 1.433>, rot=<-0.0778 -0.1601 -0.0379 -0.9833>}
 FACIAL_R_LipUpper {parent=FACIAL_C_MouthUpper, loc=<-1.2117 -0.1744 1.4329>, rot=<-0.0778 0.1601 0.0379 -0.9833>}
 FACIAL_L_LipUpperOuter {parent=FACIAL_C_MouthUpper, loc=<1.8866 0.1025 0.9251>, rot=<-0.0983 -0.2294 -0.0983 -0.9634>}
 FACIAL_R_LipUpperOuter {parent=FACIAL_C_MouthUpper, loc=<-1.8866 0.1023 0.9251>, rot=<-0.0983 0.2294 0.0983 -0.9634>}
 FACIAL_L_LipCorner {parent=FACIAL_C_MouthUpper, loc=<2.263 0.3448 0.4574>, rot=<0 -0.2079 0 -0.9781>}
 FACIAL_R_LipCorner {parent=FACIAL_C_MouthUpper, loc=<-2.2081 0.6251 0.323>, rot=<0.014 0.2112 0.0195 -0.9772>}
 FACIAL_L_Masseter {parent=FACIAL_C_FacialRoot, loc=<6.2463 1.2804 3.0238>, rot=<0.0703 0.1129 0.0071 -0.9911>}
 FACIAL_R_Masseter {parent=FACIAL_C_FacialRoot, loc=<-6.2434 1.2768 3.0245>, rot=<0.0701 -0.1136 -0.0078 -0.991>}
 FACIAL_C_TeethUpper {parent=FACIAL_C_FacialRoot, loc=<0.0003 4.9493 7.0966>, rot=<0 0 0.0001 -1>}
 FACIAL_C_LowerLipRotation {parent=FACIAL_C_FacialRoot, loc=<0 2.0037 1.1413>, rot=<0.342 0 0 -0.9397>}
 FACIAL_C_MouthLower {parent=FACIAL_C_LowerLipRotation, loc=<-0.0003 -1.2895 5.3776>, rot=<-0.342 0 0 -0.9397>}
 FACIAL_C_LipLower {parent=FACIAL_C_MouthLower, loc=<-0.0002 1.5593 1.2065>, rot=<0.0871 0 0 -0.9962>}
 FACIAL_L_LipLower {parent=FACIAL_C_MouthLower, loc=<0.9392 1.4183 1.0449>, rot=<0.0927 -0.1295 0.0273 -0.9869>}
 FACIAL_R_LipLower {parent=FACIAL_C_MouthLower, loc=<-0.9396 1.4182 1.0449>, rot=<0.0927 0.1295 -0.0273 -0.9869>}
 FACIAL_L_LipLowerOuter {parent=FACIAL_C_MouthLower, loc=<1.8086 0.8057 0.7116>, rot=<0.0913 -0.2322 0.0318 -0.9679>}
 FACIAL_R_LipLowerOuter {parent=FACIAL_C_MouthLower, loc=<-1.8088 0.8055 0.7115>, rot=<0.0913 0.2322 -0.0318 -0.9679>}
 FACIAL_C_TeethLower {parent=FACIAL_C_LowerLipRotation, loc=<-0.0003 -1.5064 6.4215>, rot=<-0.342 0 0 -0.9397>}
 FACIAL_C_Tongue1 {parent=FACIAL_C_TeethLower, loc=<-0.0001 0.3577 -4.8093>, rot=<-0.2588 0 0 -0.9659>}
 FACIAL_C_Tongue2 {parent=FACIAL_C_Tongue1, loc=<0.0001 0.3274 1.3992>, rot=<0.2588 0 0 -0.9659>}
 FACIAL_C_Tongue3 {parent=FACIAL_C_Tongue2, loc=<0 0.1493 1.2828>, rot=<0.1305 0 0 -0.9914>}
 FACIAL_C_Tongue4 {parent=FACIAL_C_Tongue3, loc=<0.0007 -0.219 1.0673>, rot=<0.0436 0 0 -0.999>}
 FACIAL_C_TongueUpper3 {parent=FACIAL_C_Tongue3, loc=<0.0001 -0.3796 -0.1323>, rot=<0 0 0 -1>}
 FACIAL_C_TongueLower3 {parent=FACIAL_C_Tongue3, loc=<-0.0001 0.7933 0.1663>, rot=<0 0 0 -1>}
 FACIAL_L_TongueSide3 {parent=FACIAL_C_Tongue3, loc=<1.5817 0.0655 0.2162>, rot=<0 0 0 -1>}
 FACIAL_R_TongueSide3 {parent=FACIAL_C_Tongue3, loc=<-1.5817 0.0654 0.2161>, rot=<0 0 0 -1>}
 FACIAL_C_TongueUpper2 {parent=FACIAL_C_Tongue2, loc=<0.0001 -0.5563 0.0821>, rot=<0 0 0 -1>}
 FACIAL_L_TongueSide2 {parent=FACIAL_C_Tongue2, loc=<1.847 0.2256 0.4873>, rot=<0 0 0 -1>}
 FACIAL_R_TongueSide2 {parent=FACIAL_C_Tongue2, loc=<-1.847 0.2255 0.4872>, rot=<0 0 0 -1>}
 FACIAL_L_Ear {parent=FACIAL_C_FacialRoot, loc=<6.3224 -0.2425 0.1892>, rot=<-0.0262 -0.0007 -0.0262 -0.9993>}
 FACIAL_R_Ear {parent=FACIAL_C_FacialRoot, loc=<-6.3225 -0.2423 0.1892>, rot=<-0.0262 0.0007 0.0262 -0.9993>}
 FACIAL_C_NoseBridge {parent=FACIAL_C_FacialRoot, loc=<0.0001 -1.5916 8.1137>, rot=<0.0436 0 0.0001 -0.999>}
 FACIAL_L_EyelidUpperB {parent=FACIAL_C_FacialRoot, loc=<2.9262 -1.5616 5.6195>, rot=<-0.225 0 0.0001 -0.9744>}
 FACIAL_L_EyelidLowerB {parent=FACIAL_C_FacialRoot, loc=<2.9262 -1.5616 5.6195>, rot=<0.3173 0 0.0001 -0.9483>}
 FACIAL_R_EyelidLowerB {parent=FACIAL_C_FacialRoot, loc=<-2.9262 -1.5609 5.6195>, rot=<0.3256 0 0.0001 -0.9455>}
 FACIAL_R_EyelidUpperB {parent=FACIAL_C_FacialRoot, loc=<-2.9262 -1.5609 5.6195>, rot=<-0.225 0 0.0001 -0.9744>}
 base_hair_fr_01 {parent=head, loc=<8.3555 7.4029 3.4137>, rot=<-0.3898 0.1224 -0.4021 -0.8194>}
 dyn_hair_fr_01 {parent=base_hair_fr_01, loc=<0 0 0>, rot=<0 0 0 1>}
 dyn_hair_fr_02 {parent=dyn_hair_fr_01, loc=<2.9327 0.0088 -0.0042>, rot=<-0.0752 0.0264 -0.8306 -0.5511>}
 dyn_hair_fr_03 {parent=dyn_hair_fr_02, loc=<6.847 0.0092 0.0108>, rot=<0.0163 -0.1312 -0.286 -0.9491>}
 dyn_hair_fr_04 {parent=dyn_hair_fr_03, loc=<6.746 0.0125 -0.0314>, rot=<0 0 0 -1>}
 attach_backpack {parent=spine_05, loc=<-0.1213 -13.469 -0.0008>, rot=<-0.0093 -0.707 0.0093 -0.707>}
 clavicle_l {parent=spine_05, loc=<5.4831 -0.1053 -3.2946>, rot=<0.0465 -0.7289 -0.009 -0.6829>}
 upperarm_l {parent=clavicle_l, loc=<9.8658 -1.2136 -0.4105>, rot=<0.0164 -0.3245 -0.0269 -0.9454>}
 lowerarm_l {parent=upperarm_l, loc=<26.465 -0.8345 0.259>, rot=<0.0036 0.0161 -0.2298 -0.9731>}
 hand_l {parent=lowerarm_l, loc=<22.7355 -0.3203 -0.6105>, rot=<-0.6784 -0.0223 0.0735 -0.7307>}
 hand_attach_l {parent=hand_l, loc=<8.5002 -3 0>, rot=<0.7071 0.7071 0 -0.0002>}
 weapon_l {parent=hand_l, loc=<-0.1973 -3.411 0.3812>, rot=<0 0 0 -1>}
 thumb_01_l {parent=hand_l, loc=<1.3118 -0.3191 -2.459>, rot=<0.6546 -0.4042 0.0664 -0.6354>}
 thumb_02_l {parent=thumb_01_l, loc=<5.9336 -0.0448 0.0401>, rot=<-0.0037 0.0213 0.298 -0.9543>}
 thumb_03_l {parent=thumb_02_l, loc=<2.2223 0.3682 -0.0907>, rot=<0.0118 -0.013 0.0066 -0.9998>}
 index_metacarpal_l {parent=hand_l, loc=<1.6535 0.0183 -2.96>, rot=<0.0158 -0.0124 -0.0173 -0.9996>}
 index_01_l {parent=index_metacarpal_l, loc=<8.3008 0.0357 -0.0524>, rot=<-0.0044 0.0485 0.0924 -0.9945>}
 index_02_l {parent=index_01_l, loc=<3.3587 -0.3004 0.0452>, rot=<-0.002 0.0027 0.3225 -0.9466>}
 index_03_l {parent=index_02_l, loc=<2.5578 -0.123 -0.0252>, rot=<-0.002 0.0067 0.2558 -0.9667>}
 middle_metacarpal_l {parent=hand_l, loc=<2.1856 -0.1725 -0.5984>, rot=<0.006 0.0168 -0.0329 -0.9993>}
 middle_01_l {parent=middle_metacarpal_l, loc=<7.3275 -0.2447 0.0129>, rot=<-0.0208 0.0042 0.0858 -0.9961>}
 middle_02_l {parent=middle_01_l, loc=<4.0123 0.1401 0.1175>, rot=<-0.0218 0.0258 0.3512 -0.9357>}
 middle_03_l {parent=middle_02_l, loc=<2.6077 0.0534 0.1373>, rot=<0.0082 0.0188 0.1708 -0.9851>}
 ring_metacarpal_l {parent=hand_l, loc=<2.2074 -0.3625 1.3638>, rot=<-0.0284 0.0722 -0.0458 -0.9959>}
 ring_01_l {parent=ring_metacarpal_l, loc=<7.0295 -0.6197 -0.1446>, rot=<-0.0489 -0.0252 0.1892 -0.9804>}
 ring_02_l {parent=ring_01_l, loc=<3.6174 0.4647 -0.1027>, rot=<0 0 0.2806 -0.9598>}
 ring_03_l {parent=ring_02_l, loc=<2.4152 0.114 0.0199>, rot=<0 0 0.1189 -0.9929>}
 pinky_metacarpal_l {parent=hand_l, loc=<1.7616 -0.9753 3.3814>, rot=<-0.0397 0.0814 -0.069 -0.9935>}
 pinky_01_l {parent=pinky_metacarpal_l, loc=<7.1709 -0.4831 -0.1094>, rot=<-0.0654 -0.0035 0.2242 -0.9723>}
 pinky_02_l {parent=pinky_01_l, loc=<2.8274 0.0893 -0.1043>, rot=<0.0037 0.0011 0.2854 -0.9584>}
 pinky_03_l {parent=pinky_02_l, loc=<2.2046 0.0307 -0.0078>, rot=<-0.0002 0.0098 0.1445 -0.9895>}
 deform_wrist_dn_base_l {parent=hand_l, loc=<0 -1.5326 0>, rot=<0 0 0 -1>}
 deform_wrist_dn_l {parent=deform_wrist_dn_base_l, loc=<0 0 0>, rot=<0 0 0 1>}
 deform_wrist_up_base_l {parent=hand_l, loc=<0 2.3009 0>, rot=<0 0 0 -1>}
 deform_wrist_up_l {parent=deform_wrist_up_base_l, loc=<0 0 0>, rot=<0 0 0 1>}
 lowerarm_twist_02_l {parent=lowerarm_l, loc=<10.6707 0 0>, rot=<0 0 0 1>}
 lowerarm_twist_01_l {parent=lowerarm_l, loc=<18.316 0 0>, rot=<0 0 0 1>}
 deform_elbow_in_base_l {parent=lowerarm_l, loc=<0.1834 3.6062 -0.2572>, rot=<0 0 0 -1>}
 deform_elbow_in_l {parent=deform_elbow_in_base_l, loc=<0 0 0>, rot=<0 0 0 1>}
 deform_elbow_out_base_l {parent=lowerarm_l, loc=<0 -4 0>, rot=<0 0 0 1>}
 deform_elbow_out_l {parent=deform_elbow_out_base_l, loc=<0 0 0>, rot=<0 0 0 1>}
 upperarm_twist_01_l {parent=upperarm_l, loc=<7.2127 -0.1129 0.0005>, rot=<0 0 0 -1>}
 deform_upperarm_bk_base_l {parent=upperarm_twist_01_l, loc=<-6.4047 -4.375 0.3136>, rot=<0 0 0 -1>}
 deform_upperarm_bk_l {parent=deform_upperarm_bk_base_l, loc=<0 0 0>, rot=<0 0 0 1>}
 deform_upperarm_fr_base_l {parent=upperarm_twist_01_l, loc=<-5.6306 5.0916 0.3348>, rot=<0 0 0 -1>}
 deform_upperarm_fr_l {parent=deform_upperarm_fr_base_l, loc=<0 0 0>, rot=<0 0 0 1>}
 deform_upperarm_up_base_l {parent=upperarm_twist_01_l, loc=<-8.8907 0.4823 3.9514>, rot=<0 0 0 -1>}
 deform_upperarm_up_l {parent=deform_upperarm_up_base_l, loc=<0 0 0>, rot=<0 0 0 1>}
 upperarm_twist_01_bffix_l {parent=upperarm_twist_01_l, loc=<0 0 0>, rot=<0 0 0 1>}
 upperarm_twist_02_l {parent=upperarm_l, loc=<19.7998 -0.1134 0>, rot=<0 0 0 1>}
 base_jacket_a_01_l {parent=upperarm_l, loc=<3.3401 -0.4311 4.7453>, rot=<-0.0192 0.3207 0.0404 -0.9461>}
 dyn_jacket_a_01_l {parent=base_jacket_a_01_l, loc=<0 0 0>, rot=<0 0 0 1>}
 dyn_jacket_a_02_l {parent=dyn_jacket_a_01_l, loc=<3.4111 -0.0085 1.5198>, rot=<0 0 0 1>}
 base_jacket_b_01_l {parent=upperarm_l, loc=<5.8618 4.6155 0.1836>, rot=<0.1893 0.2994 -0.5111 -0.7831>}
 dyn_jacket_b_01_l {parent=base_jacket_b_01_l, loc=<0 0 0>, rot=<0 0 0 1>}
 dyn_jacket_b_02_l {parent=dyn_jacket_b_01_l, loc=<3.6777 -0.0004 -0.0001>, rot=<0 0 0 1>}
 clavicle_r {parent=spine_05, loc=<5.4831 -0.1053 3.2946>, rot=<0.7289 0.0465 -0.6829 0.009>}
 upperarm_r {parent=clavicle_r, loc=<-9.8658 1.2136 0.4105>, rot=<0.0164 -0.3245 -0.0269 -0.9454>}
 lowerarm_r {parent=upperarm_r, loc=<-26.465 0.8345 -0.259>, rot=<0.0036 0.0161 -0.2298 -0.9731>}
 hand_r {parent=lowerarm_r, loc=<-22.7355 0.3203 0.6106>, rot=<-0.6784 -0.0223 0.0735 -0.7307>}
 thumb_01_r {parent=hand_r, loc=<-1.3118 0.3191 2.459>, rot=<0.6546 -0.4042 0.0664 -0.6354>}
 thumb_02_r {parent=thumb_01_r, loc=<-5.9336 0.0448 -0.0401>, rot=<-0.0037 0.0213 0.298 -0.9543>}
 thumb_03_r {parent=thumb_02_r, loc=<-2.2223 -0.3682 0.0907>, rot=<0.0118 -0.013 0.0066 -0.9998>}
 index_metacarpal_r {parent=hand_r, loc=<-1.6535 -0.0183 2.96>, rot=<0.0158 -0.0124 -0.0173 -0.9996>}
 index_01_r {parent=index_metacarpal_r, loc=<-8.3008 -0.0357 0.0524>, rot=<-0.0044 0.0485 0.0924 -0.9945>}
 index_02_r {parent=index_01_r, loc=<-3.3587 0.3004 -0.0452>, rot=<-0.002 0.0027 0.3225 -0.9466>}
 index_03_r {parent=index_02_r, loc=<-2.5578 0.123 0.0252>, rot=<-0.002 0.0067 0.2558 -0.9667>}
 middle_metacarpal_r {parent=hand_r, loc=<-2.1856 0.1725 0.5984>, rot=<0.006 0.0168 -0.0329 -0.9993>}
 middle_01_r {parent=middle_metacarpal_r, loc=<-7.3275 0.2447 -0.0129>, rot=<-0.0208 0.0042 0.0858 -0.9961>}
 middle_02_r {parent=middle_01_r, loc=<-4.0123 -0.1401 -0.1175>, rot=<-0.0218 0.0258 0.3512 -0.9357>}
 middle_03_r {parent=middle_02_r, loc=<-2.6077 -0.0534 -0.1373>, rot=<0.0082 0.0188 0.1708 -0.9851>}
 ring_metacarpal_r {parent=hand_r, loc=<-2.2074 0.3625 -1.3638>, rot=<-0.0284 0.0722 -0.0458 -0.9959>}
 ring_01_r {parent=ring_metacarpal_r, loc=<-7.0295 0.6197 0.1446>, rot=<-0.0489 -0.0252 0.1892 -0.9804>}
 ring_02_r {parent=ring_01_r, loc=<-3.6174 -0.4647 0.1027>, rot=<0 0 0.2806 -0.9598>}
 ring_03_r {parent=ring_02_r, loc=<-2.4152 -0.114 -0.0199>, rot=<0 0 0.1189 -0.9929>}
 pinky_metacarpal_r {parent=hand_r, loc=<-1.7616 0.9753 -3.3814>, rot=<-0.0397 0.0814 -0.069 -0.9935>}
 pinky_01_r {parent=pinky_metacarpal_r, loc=<-7.1709 0.4831 0.1094>, rot=<-0.0654 -0.0035 0.2242 -0.9723>}
 pinky_02_r {parent=pinky_01_r, loc=<-2.8274 -0.0893 0.1043>, rot=<0.0037 0.0011 0.2854 -0.9584>}
 pinky_03_r {parent=pinky_02_r, loc=<-2.2046 -0.0307 0.0078>, rot=<-0.0002 0.0098 0.1445 -0.9895>}
 weapon_r {parent=hand_r, loc=<1.9909 -0.2362 -0.9788>, rot=<-0.0219 -0.0516 0.9964 0.0636>}
 hand_attach_r {parent=hand_r, loc=<-8.5003 3 0.0001>, rot=<0 0 -0.7071 -0.7071>}
 deform_wrist_dn_base_r {parent=hand_r, loc=<-0.0001 1.5328 0>, rot=<0 0 0 -1>}
 deform_wrist_dn_r {parent=deform_wrist_dn_base_r, loc=<0 0 0>, rot=<0 0 0 1>}
 deform_wrist_up_base_r {parent=hand_r, loc=<-0.0002 -2.3007 -0.0001>, rot=<0 0 0 -1>}
 deform_wrist_up_r {parent=deform_wrist_up_base_r, loc=<0 0 0>, rot=<0 0 0 1>}
 lowerarm_twist_02_r {parent=lowerarm_r, loc=<-10.6707 0 0>, rot=<0 0 0 -1>}
 lowerarm_twist_01_r {parent=lowerarm_r, loc=<-18.316 0 0>, rot=<0 0 0 -1>}
 deform_elbow_in_base_r {parent=lowerarm_r, loc=<-0.1834 -3.6062 0.2572>, rot=<0 0 0 -1>}
 deform_elbow_in_r {parent=deform_elbow_in_base_r, loc=<0 0 0>, rot=<0 0 0 1>}
 deform_elbow_out_base_r {parent=lowerarm_r, loc=<0 4 -0.0001>, rot=<0 0 0 -1>}
 deform_elbow_out_r {parent=deform_elbow_out_base_r, loc=<0 0 0>, rot=<0 0 0 1>}
 upperarm_twist_01_r {parent=upperarm_r, loc=<-7.2128 0.1131 0.0002>, rot=<0 0 0 -1>}
 deform_upperarm_bk_base_r {parent=upperarm_twist_01_r, loc=<6.4048 4.3747 -0.3142>, rot=<0 0 0 -1>}
 deform_upperarm_bk_r {parent=deform_upperarm_bk_base_r, loc=<0 0 0>, rot=<0 0 0 1>}
 deform_upperarm_fr_base_r {parent=upperarm_twist_01_r, loc=<5.6307 -5.0919 -0.3354>, rot=<0 0 0 -1>}
 deform_upperarm_fr_r {parent=deform_upperarm_fr_base_r, loc=<0 0 0>, rot=<0 0 0 1>}
 deform_upperarm_up_base_r {parent=upperarm_twist_01_r, loc=<8.8907 -0.4825 -3.9518>, rot=<0 0 0 -1>}
 deform_upperarm_up_r {parent=deform_upperarm_up_base_r, loc=<0 0 0>, rot=<0 0 0 1>}
 upperarm_twist_01_bffix_r {parent=upperarm_twist_01_r, loc=<0 0 0>, rot=<0 0 0 -1>}
 upperarm_twist_02_r {parent=upperarm_r, loc=<-19.7998 0.1134 0>, rot=<0 0 0 -1>}
 base_jacket_a_01_r {parent=upperarm_r, loc=<-3.3395 0.4316 -4.7458>, rot=<-0.3207 -0.0192 0.9461 0.0404>}
 dyn_jacket_a_01_r {parent=base_jacket_a_01_r, loc=<0 0 0>, rot=<0 0 0 1>}
 dyn_jacket_a_02_r {parent=dyn_jacket_a_01_r, loc=<3.3403 -0.0088 -1.5202>, rot=<0 0 0 1>}
 base_jacket_b_01_r {parent=upperarm_r, loc=<-5.8615 -4.6159 -0.1827>, rot=<-0.2994 0.1893 0.7831 -0.5111>}
 dyn_jacket_b_01_r {parent=base_jacket_b_01_r, loc=<0 0 0>, rot=<0 0 0 1>}
 dyn_jacket_b_02_r {parent=dyn_jacket_b_01_r, loc=<3.6778 0 0.0001>, rot=<0 0 0 1>}
 base_hair_back_aimtarget {parent=spine_05, loc=<1.9279 -5.3482 9.3509>, rot=<0.0284 -0.7065 -0.0284 -0.7065>}
 deform_pec_aim_base_l {parent=spine_04, loc=<3.2456 9.3309 -1.2237>, rot=<0.0603 -0.4033 0.3318 -0.8507>}
 deform_pec_aim_l {parent=deform_pec_aim_base_l, loc=<0 0 0>, rot=<0 0 0 1>}
 deform_pec_base_l {parent=deform_pec_aim_l, loc=<9.3665 0 0>, rot=<0 0 0 1>}
 deform_pec_l {parent=deform_pec_base_l, loc=<0 0 0>, rot=<0 0 0 1>}
 deform_pec_aim_base_r {parent=spine_04, loc=<3.2452 9.3309 1.2237>, rot=<-0.4033 -0.0603 0.8507 0.3318>}
 deform_pec_aim_r {parent=deform_pec_aim_base_r, loc=<0 0 0>, rot=<0 0 0 1>}
 deform_pec_base_r {parent=deform_pec_aim_r, loc=<-9.3667 -0.0003 -0.0004>, rot=<0 0 0 1>}
 deform_pec_r {parent=deform_pec_base_r, loc=<0 0 0>, rot=<0 0 0 1>}
 base_jacket_c_01_r {parent=spine_04, loc=<-0.6512 2.863 13.7265>, rot=<-0.6391 -0.7685 -0.0207 0.0246>}
 dyn_jacket_c_01_r {parent=base_jacket_c_01_r, loc=<0 0 0>, rot=<0 0 0 1>}
 dyn_jacket_c_02_r {parent=dyn_jacket_c_01_r, loc=<4.5636 0 0>, rot=<0 0 0 1>}
 base_jacket_c_01_l {parent=spine_04, loc=<-0.6512 2.863 -13.7265>, rot=<-0.4353 -0.5588 -0.5279 -0.4686>}
 dyn_jacket_c_01_l {parent=base_jacket_c_01_l, loc=<0 0 0>, rot=<0 0 0 1>}
 dyn_jacket_c_02_l {parent=dyn_jacket_c_01_l, loc=<4.5636 0 0>, rot=<0 0 0 1>}
 base_jacket_helper_l {parent=spine_04, loc=<3.3419 1.9061 -14.7202>, rot=<-0.4353 -0.5588 -0.5279 -0.4686>}
 dyn_jacket_helper_l {parent=base_jacket_helper_l, loc=<0 0 0>, rot=<0 0 0 -1>}
 base_jacket_helper_r {parent=spine_04, loc=<2.9851 2.1083 14.7741>, rot=<-0.4353 -0.5588 -0.5279 -0.4686>}
 dyn_jacket_helper_r {parent=base_jacket_helper_r, loc=<0 0 0>, rot=<0 0 0 -1>}
 base_jacket_d_01_r {parent=spine_03, loc=<-2.4802 3.2655 11.7662>, rot=<-0.2742 -0.9604 0.0188 0.046>}
 dyn_jacket_d_01_r {parent=base_jacket_d_01_r, loc=<0 0 0>, rot=<0 0 0 1>}
 dyn_jacket_d_02_r {parent=dyn_jacket_d_01_r, loc=<5.2891 0.8394 0>, rot=<0 0 0 1>}
 base_jacket_d_01_l {parent=spine_03, loc=<-2.4802 3.2655 -11.7662>, rot=<-0.2742 -0.9604 -0.0188 -0.046>}
 dyn_jacket_d_01_l {parent=base_jacket_d_01_l, loc=<0 0 0>, rot=<0 0 0 1>}
 dyn_jacket_d_02_l {parent=dyn_jacket_d_01_l, loc=<5.2891 0.8394 0>, rot=<0 0 0 1>}
 base_jacket_bck_01_r {parent=spine_03, loc=<2.9681 -7.254 4.7965>, rot=<0.0878 -0.9959 -0.0004 -0.0206>}
 dyn_jacket_bck_01_r {parent=base_jacket_bck_01_r, loc=<0 0 0>, rot=<0 0 0 1>}
 base_jacket_bck_01_l {parent=spine_03, loc=<2.5926 -7.2003 -4.6538>, rot=<0.1159 -0.9932 -0.0001 0.0142>}
 dyn_jacket_bck_01_l {parent=base_jacket_bck_01_l, loc=<0 0 0>, rot=<0 0 0 1>}
 helper_spine_03 {parent=spine_03, loc=<0 0 0>, rot=<0 0 0 1>, scale=<0.8 0.8 0.8>}
 thigh_l {parent=pelvis, loc=<-1.0855 -0.2109 -8.7053>, rot=<0.0009 0.0355 0.0155 -0.9993>}
 calf_l {parent=thigh_l, loc=<-42.5284 -0.0019 -0.7297>, rot=<0.0006 -0.0161 -0.0391 -0.9991>}
 foot_l {parent=calf_l, loc=<-46.7401 -2.5014 -0.592>, rot=<0.0427 -0.0238 0.0353 -0.9982>}
 ball_l {parent=foot_l, loc=<-3.5658 10.8066 -0.1544>, rot=<0.0001 -0.0001 -0.7063 -0.7079>}
 calf_twist_02_l {parent=calf_l, loc=<-19.1546 -1.7943 -0.2638>, rot=<0.0237 0.0095 -0.0035 -0.9997>}
 calf_twist_01_l {parent=calf_l, loc=<-35.8483 -1.9197 -0.5794>, rot=<0.0237 0.0095 -0.0035 -0.9997>}
 deform_knee_in_base_l {parent=calf_l, loc=<0.8031 -5.0231 0.2456>, rot=<0 0 0 -1>}
 deform_knee_in_l {parent=deform_knee_in_base_l, loc=<0 0 0>, rot=<0 0 0 1>}
 deform_knee_out_base_l {parent=calf_l, loc=<0 4 0>, rot=<0 0 0 -1>}
 deform_knee_out_l {parent=deform_knee_out_base_l, loc=<0 0 0>, rot=<0 0 0 1>}
 thigh_twist_01_l {parent=thigh_l, loc=<-27.952 0 0>, rot=<0.0013 0 0 -1>}
 deform_groin_base_l {parent=thigh_l, loc=<0.4801 5.6776 -0.2486>, rot=<0 0 0 -1>}
 deform_groin_l {parent=deform_groin_base_l, loc=<0 0 0>, rot=<0 0 0 1>}
 thigh_r {parent=pelvis, loc=<-1.0852 -0.2106 8.7051>, rot=<0.0355 -0.0009 0.9993 0.0155>}
 calf_r {parent=thigh_r, loc=<42.5284 0.0019 0.7297>, rot=<0.0006 -0.0161 -0.0391 -0.9991>}
 foot_r {parent=calf_r, loc=<46.7401 2.5014 0.592>, rot=<0.0427 -0.0238 0.0353 -0.9982>}
 ball_r {parent=foot_r, loc=<3.5658 -10.8066 0.1544>, rot=<0.0001 -0.0001 -0.7063 -0.7079>}
 calf_twist_02_r {parent=calf_r, loc=<19.1546 1.7943 0.2638>, rot=<0.0237 0.0095 -0.0035 -0.9997>}
 calf_twist_01_r {parent=calf_r, loc=<35.8483 1.9197 0.5794>, rot=<0.0237 0.0095 -0.0035 -0.9997>}
 deform_knee_in_base_r {parent=calf_r, loc=<-0.8031 5.0231 -0.2457>, rot=<0 0 0 -1>}
 deform_knee_in_r {parent=deform_knee_in_base_r, loc=<0 0 0>, rot=<0 0 0 1>}
 deform_knee_out_base_r {parent=calf_r, loc=<0 -4 0>, rot=<0 0 0 -1>}
 deform_knee_out_r {parent=deform_knee_out_base_r, loc=<0 0 0>, rot=<0 0 0 1>}
 thigh_twist_01_r {parent=thigh_r, loc=<27.952 0 0>, rot=<0.0013 0 0 -1>}
 base_chain {parent=thigh_r, loc=<0 0 0>, rot=<0.0355 -0.0009 0.9993 -0.0155>}
 base_chain_fr {parent=base_chain, loc=<2.1236 8.551 -8.0271>, rot=<-0.3154 -0.673 -0.6354 0.2094>}
 dyn_chain_fr_00 {parent=base_chain_fr, loc=<0 0 0>, rot=<0 0 0 1>}
 dyn_chain_fr_01 {parent=dyn_chain_fr_00, loc=<7.7749 0 0>, rot=<0 -0.0604 0 -0.9982>}
 dyn_chain_fr_02 {parent=dyn_chain_fr_01, loc=<7.4515 0.3678 0.0781>, rot=<0.1571 -0.1567 0.0252 -0.9748>}
 dyn_chain_fr_03 {parent=dyn_chain_fr_02, loc=<7.4351 1.274 0.3685>, rot=<0 0 0 1>}
 base_chain_bck {parent=base_chain, loc=<6.5055 -9.2958 -9.1834>, rot=<0.32 0.6308 0.6395 -0.3012>}
 dyn_chain_bck_00 {parent=base_chain_bck, loc=<0 0 0>, rot=<0 0 0 1>}
 dyn_chain_bck_01 {parent=dyn_chain_bck_00, loc=<11.7943 -0.36 -3.0672>, rot=<0 0.096 0 -0.9954>}
 dyn_chain_bck_02 {parent=dyn_chain_bck_01, loc=<9.5771 -0.5187 -0.2469>, rot=<0.0501 0.1415 -0.0072 -0.9886>}
 dyn_chain_bck_03 {parent=dyn_chain_bck_02, loc=<6.9243 -1.1697 -0.2484>, rot=<0 0 0 1>}
 deform_groin_base_r {parent=thigh_r, loc=<-0.4801 -5.6776 0.2486>, rot=<0 0 0 -1>}
 deform_groin_r {parent=deform_groin_base_r, loc=<0 0 0>, rot=<0 0 0 1>}
 base_chain_acc_bck_01 {parent=thigh_r, loc=<12.6387 7.5272 10.2746>, rot=<-0.5513 -0.1269 -0.3136 -0.7626>}
 dyn_chain_acc_bck_02 {parent=base_chain_acc_bck_01, loc=<0 0 0>, rot=<0 0 0 1>}
 base_chain_acc_cn_01 {parent=thigh_r, loc=<17.0957 -0.5101 11.646>, rot=<0.9665 0.0235 0.133 -0.2182>}
 dyn_chain_acc_cn_02 {parent=base_chain_acc_cn_01, loc=<0 0 0>, rot=<0 0 0 1>}
 base_chain_acc_fr_01 {parent=thigh_r, loc=<14.5887 -7.8431 6.7468>, rot=<0.9211 -0.2487 -0.1362 0.2669>}
 dyn_chain_acc_fr_02 {parent=base_chain_acc_fr_01, loc=<0 0 0>, rot=<0 0 0 1>}
 helper_thigh_r {parent=thigh_r, loc=<-3.7761 2.1314 0.0014>, rot=<0 0 0.0182 -0.9998>}
 deform_glute_base_l {parent=pelvis, loc=<-7.3466 -9.0213 -9.1245>, rot=<0.0172 0.031 -0.4477 -0.8935>}
 deform_glute_l {parent=deform_glute_base_l, loc=<0 0 0>, rot=<0 0 0 1>}
 deform_glute_base_r {parent=pelvis, loc=<-7.3464 -9.0213 9.1245>, rot=<0.031 -0.0172 0.8935 -0.4477>}
 deform_glute_r {parent=deform_glute_base_r, loc=<0 0 0>, rot=<0 0 0 1>}
 thigh_r_null {parent=pelvis, type=Null, loc=<-1.0855 -0.2109 8.7053>, rot=<0.0355 -0.0009 0.9993 0.0155>}
 aim_chain_acc_cn_02 {parent=pelvis, loc=<-17.2906 0.793 21.5352>, rot=<-0.0164 0.9616 -0.2143 -0.1706>}
 aim_chain_acc_cn_02_2 {parent=aim_chain_acc_cn_02, loc=<0 0 0>, rot=<0 0 0 1>}
 aim_chain_fr_01 {parent=pelvis, loc=<-4.5078 9.5719 5.986>, rot=<-0.2765 -0.6591 -0.6533 0.2496>}
 aim_chain_bck_01 {parent=pelvis, loc=<-3.518 -12.0372 10.3276>, rot=<0.3799 0.6568 0.6058 -0.2393>}
 thigh_l_null {parent=pelvis, type=Null, loc=<-1.0855 -0.2109 -8.7053>, rot=<0.0009 0.0355 0.0155 -0.9993>}
 aim_chain_front_chain {parent=pelvis, loc=<-14.4867 7.926 16.8226>, rot=<0.1897 0.5077 0.79 -0.2867>}
 aim_chain_back_chain {parent=pelvis, loc=<-12.8779 -7.374 20.0693>, rot=<-0.478 -0.6501 -0.5764 0.1289>}
 attach {parent=root, loc=<0 0 0>, rot=<0 0 0 1>}
 ik_hand_root {parent=root, loc=<0 0 0>, rot=<0 0 0 1>}
 ik_hand_gun {parent=ik_hand_root, loc=<-48.1743 3.3545 104.4381>, rot=<0.6276 0.1309 0.4183 -0.6434>}
 ik_hand_r {parent=ik_hand_gun, loc=<0 0 0>, rot=<0 0 0 1>}
 ik_hand_l {parent=ik_hand_gun, loc=<59.3271 67.6997 34.3524>, rot=<-0.6158 -0.7027 -0.3565 0>}
 ik_foot_root {parent=root, loc=<0 0 0>, rot=<0 0 0 1>}
 ik_foot_l {parent=ik_foot_root, loc=<14.8262 -7.3545 8.296>, rot=<-0.0364 0.7015 -0.0316 -0.711>}
 ik_foot_r {parent=ik_foot_root, loc=<-14.8262 -7.3545 8.296>, rot=<-0.6616 -0.2497 0.6616 -0.2496>}
 root_fp {parent=root, loc=<0 0 0>, rot=<0 0 0 1>}
 attach_fp {parent=root_fp, loc=<0 0 0>, rot=<0 0 0 1>}
 camera_root {parent=root_fp, loc=<0 0 0>, rot=<0 0 0 1>}
 camera {parent=camera_root, loc=<0 0 0>, rot=<0 0 0 1>}
 attach_emote_01 {parent=root, loc=<0 0 0>, rot=<0 0 0 1>}
 attach_emote_02 {parent=root, loc=<0 0 0>, rot=<0 0 0 1>}
 attach_emote_03 {parent=root, loc=<0 0 0>, rot=<0 0 0 1>}
 warp_point_01 {parent=root, loc=<0 0 0>, rot=<0 0 0 1>}
 warp_point_02 {parent=root, loc=<0 0 0>, rot=<0 0 0 1>}
 warp_point_03 {parent=root, loc=<0 0 0>, rot=<0 0 0 1>}
 spine_01_ctrl {controlType=EULER_TRANSFORM, type=Control, loc=<0 0 0>, rot=<0 0 0 1>}
 thigh_r_ctrl {controlType=EULER_TRANSFORM, type=Control, loc=<0 0 0>, rot=<0 0 0 1>}
 upperarm_l_ctrl {controlType=EULER_TRANSFORM, type=Control, loc=<-98.5329 0.2842 -112.7497>, rot=<0.0143 -0.707 -0.0115 0.707>}
 upperarm_r_ctrl {controlType=EULER_TRANSFORM, type=Control, loc=<-121.0609 1.8575 -48.3191>, rot=<0.0027 0.7067 0.0267 0.707>}
]
variables [swimming {type=bool, default=False}]
