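# Physics asset for the hanging chain on the female medium ElegantHeist body, with capsule bodies for the chain, pelvis and right thigh.
# Front and back four-bone strands start at a base locked to the pelvis and join at their tips through a constraint with free rotation, closing a loop.
# Root links are locked, and swing loosens down each strand.
# Pelvis and rigidly attached right-thigh capsules give the body volume around the chain.
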PhysicsAsset /BRCosmetics/Characters/Player/Female/Medium/Bodies/F_MED_ElegantHeist/Meshes/F_MED_ElegantHeist_Physics_Chain
bodies {
 base_chain [{shapeType=Capsule, length=0.505, radius=0.505, shapeName=base_chain_capsule, rotation=<rot 0 -90 0>}]
 dyn_chain_bck_00 [{
   shapeType=Capsule
   length=5.7516
   radius=0.8429
   shapeName=dyn_chain_bck_00_capsule
   center=<6.6615 -0.0929 -3.1719>
   rotation=<rot 0 -89.2959 -78.0663>
  }]
 dyn_chain_bck_01 [{
   shapeType=Capsule
   length=5.5109
   radius=2.7061
   shapeName=dyn_chain_bck_01_capsule
   center=<3.511 0.0067 -2.9261>
   rotation=<rot -0.0001 89.7166 -98.7014>
  }]
 dyn_chain_bck_02 [{shapeType=Capsule, length=5, radius=2.2584, shapeName=dyn_chain_bck_02_capsule, center=<3.4732 0 -2.0941>, rotation=<rot 90 0 0>}]
 dyn_chain_bck_03 [{shapeType=Capsule, length=7, radius=2.2361, shapeName=dyn_chain_bck_03_capsule, center=<4.298 -1.7 0.6>, rotation=<rot 0 69.5835 -54.8077>}]
 dyn_chain_fr_00 [{
   shapeType=Capsule
   length=2.1653
   radius=0.7113
   shapeName=dyn_chain_fr_00_capsule
   center=<1.9931 0.3419 -0.0926>
   rotation=<rot 0 -96.0844 -86.3885>
  }]
 dyn_chain_fr_01 [{shapeType=Capsule, length=8.4554, radius=1.5, shapeName=dyn_chain_fr_01_capsule, center=<3.067 0 0.5598>, rotation=<rot 90 0 0>}]
 dyn_chain_fr_02 [{shapeType=Capsule, length=4, radius=1.5, shapeName=dyn_chain_fr_02_capsule, center=<3.6041 0 1.7052>, rotation=<rot 90 0 0>}]
 dyn_chain_fr_03 [{shapeType=Capsule, length=7, radius=2.1031, shapeName=dyn_chain_fr_03_capsule, center=<4.6743 2.5 -0.5>, rotation=<rot 0 -56.443 -60.21>}]
 helper_thigh_r [{shapeType=Capsule, length=27.1619, radius=9.9159, center=<20.9705 -1.2815 1.0745>, rotation=<rot -0.1338 -91.4159 -85.5651>}]
 pelvis [{shapeType=Capsule, length=14, radius=10.5, center=<-6.4911 -2.0766 -0.0057>}]
}
constraints [
 {
  bone1Name=base_chain
  bone2Name=pelvis
  swing1LimitDegrees=45
  swing1Motion=Locked
  swing2LimitDegrees=45
  swing2Motion=Locked
  twistLimitDegrees=45
  twistMotion=Locked
 }
 {
  bone1Name=dyn_chain_fr_00
  bone2Name=base_chain
  swing1LimitDegrees=45
  swing1Motion=Locked
  swing2LimitDegrees=45
  swing2Motion=Locked
  twistLimitDegrees=45
  twistMotion=Locked
 }
 {bone1Name=dyn_chain_fr_02, bone2Name=dyn_chain_fr_01, swing1LimitDegrees=45, swing2LimitDegrees=45, twistLimitDegrees=0}
 {bone1Name=dyn_chain_fr_01, bone2Name=dyn_chain_fr_00, swing1LimitDegrees=20, swing2LimitDegrees=20, twistLimitDegrees=0}
 {bone1Name=dyn_chain_bck_02, bone2Name=dyn_chain_bck_01, swing1LimitDegrees=35, swing2LimitDegrees=35, twistLimitDegrees=0}
 {
  bone1Name=helper_thigh_r
  bone2Name=pelvis
  swing1LimitDegrees=45
  swing1Motion=Locked
  swing2LimitDegrees=45
  swing2Motion=Locked
  twistLimitDegrees=45
  twistMotion=Locked
 }
 {bone1Name=dyn_chain_bck_03, bone2Name=dyn_chain_bck_02, swing1LimitDegrees=45, swing2LimitDegrees=45, twistLimitDegrees=0}
 {
  bone1Name=dyn_chain_fr_03
  bone2Name=dyn_chain_fr_02
  swing1LimitDegrees=45
  swing1Motion=Free
  swing2LimitDegrees=45
  swing2Motion=Free
  twistLimitDegrees=0
  twistMotion=Locked
 }
 {
  bone1Name=dyn_chain_bck_03
  bone2Name=dyn_chain_fr_03
  swing1LimitDegrees=45
  swing1Motion=Free
  swing2LimitDegrees=45
  swing2Motion=Free
  twistLimitDegrees=45
  twistMotion=Free
 }
 {
  bone1Name=dyn_chain_bck_00
  bone2Name=base_chain
  swing1LimitDegrees=35
  swing1Motion=Locked
  swing2LimitDegrees=35
  swing2Motion=Locked
  twistLimitDegrees=0
  twistMotion=Locked
 }
 {bone1Name=dyn_chain_bck_01, bone2Name=dyn_chain_bck_00, swing1LimitDegrees=20, swing2LimitDegrees=20, twistLimitDegrees=0}
]
