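# Physics asset for the F_MED_ElegantHeist female body, setting up capsule bodies and joints for jacket secondary motion.
# Pelvis, spine, upper-arm and thigh capsules approximate the body that the jacket bones hang from or around.
# Mirrored two-link jacket chains use tight angular limits: a and b attach to the upper arms, c to spine_04, d to spine_03.
# A single front chain capsule is angularly locked to the right thigh.
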PhysicsAsset /BRCosmetics/Characters/Player/Female/Medium/Bodies/F_MED_ElegantHeist/Meshes/F_MED_ElegantHeist_Physics_Jacket
bodies {
 dyn_chain_fr_02 [{
   shapeType=Capsule
   length=11.2589
   radius=2.2043
   shapeName=dyn_chain_fr_02_capsule
   center=<3.9775 0.5999 0.4725>
   rotation=<rot 0 100.4322 -85.2401>
  }]
 dyn_jacket_a_01_l [{shapeType=Capsule, length=0.5, radius=2.5, shapeName=dyn_jacket_a_01_l_capsule, rotation=<rot 90 0 0>}]
 dyn_jacket_a_01_r [{shapeType=Capsule, length=0.5, radius=2.5, shapeName=dyn_jacket_a_01_r_capsule, rotation=<rot 90 0 0>}]
 dyn_jacket_a_02_l [{shapeType=Capsule, length=0.5, radius=2.5, shapeName=dyn_jacket_a_02_l_capsule, center=<0.7 0 0.8103>, rotation=<rot 90 0 0>}]
 dyn_jacket_a_02_r [{shapeType=Capsule, length=0.5, radius=2.5, shapeName=dyn_jacket_a_02_r_capsule, center=<0.7 0 -1.5594>, rotation=<rot 90 0 0>}]
 dyn_jacket_b_01_l [{shapeType=Capsule, length=0, radius=3, shapeName=dyn_jacket_b_01_l_capsule, rotation=<rot 90 0 0>}]
 dyn_jacket_b_01_r [{shapeType=Capsule, length=0, radius=3, shapeName=dyn_jacket_b_01_r_capsule, rotation=<rot 90 0 0>}]
 dyn_jacket_b_02_l [{shapeType=Capsule, length=0, radius=3, shapeName=dyn_jacket_b_02_l_capsule, center=<0.7 0 0>, rotation=<rot 90 0 0>}]
 dyn_jacket_b_02_r [{shapeType=Capsule, length=0, radius=3, shapeName=dyn_jacket_b_02_r_capsule, center=<0.7 0 0>, rotation=<rot 90 0 0>}]
 dyn_jacket_c_01_l [{shapeType=Capsule, length=0.5, radius=2, shapeName=dyn_jacket_c_01_l_capsule, rotation=<rot 90 0 0>}]
 dyn_jacket_c_01_r [{shapeType=Capsule, length=0.5, radius=2, shapeName=dyn_jacket_c_01_r_capsule, rotation=<rot 90 0 0>}]
 dyn_jacket_c_02_l [{shapeType=Capsule, length=2, radius=2, shapeName=dyn_jacket_c_02_l_capsule, center=<0.5 0 0.2888>, rotation=<rot 22.4829 90.0002 0.0006>}]
 dyn_jacket_c_02_r [{shapeType=Capsule, length=2, radius=2, shapeName=dyn_jacket_c_02_r_capsule, center=<-0.5 0 0>, rotation=<rot -25.8399 0 90>}]
 dyn_jacket_d_01_l [{
   shapeType=Capsule
   length=14.5
   radius=2.5
   shapeName=dyn_jacket_d_01_l_capsule
   center=<1.3869 -0.5124 1.88>
   rotation=<rot -0.8672 -4.9241 -94.9627>
  }]
 dyn_jacket_d_01_r [{
   shapeType=Capsule
   length=14.5
   radius=2.5
   shapeName=dyn_jacket_d_01_r_capsule
   center=<1.3816 0.4829 -1.88>
   rotation=<rot -0.8672 -4.9241 -85.0373>
  }]
 dyn_jacket_d_02_l [{shapeType=Capsule, length=3, radius=3, shapeName=dyn_jacket_d_02_l_capsule, center=<4.4244 0 1.1671>, rotation=<rot 90 0 0>}]
 dyn_jacket_d_02_r [{shapeType=Capsule, length=7, radius=3.5, shapeName=dyn_jacket_d_02_r_capsule, center=<2.531 0 -1.7084>, rotation=<rot -90 0 0>}]
 pelvis [{shapeType=Capsule, length=14.5452, radius=9.7786, center=<-0.6206 -0.7175 0>}]
 spine_01 [{shapeType=Capsule, length=11.8901, radius=8.6468, center=<-6.1108 -0.7175 0>}]
 spine_03 [{shapeType=Capsule, length=7.754, radius=11.1971, center=<-4.537 2.0465 0.0117>, rotation=<rot 90 0 0>}]
 spine_04 [{shapeType=Capsule, length=1.5666, radius=12.2933, center=<0 3.8022 0>, rotation=<rot 0 -90 0>}]
 spine_05 [{shapeType=Capsule, length=6.4311, radius=10.6099, center=<-3.5961 1.2234 0.3935>, rotation=<rot 0 -90 0>}]
 thigh_l [{shapeType=Capsule, length=19, radius=10, center=<-18.069 -2.5999 0>, rotation=<rot -90 0 0>}]
 thigh_r [{shapeType=Capsule, length=19, radius=10, center=<20.3454 2.1506 0>, rotation=<rot -90 180 180>}]
 upperarm_l [{shapeType=Capsule, length=17.4826, radius=5.5477, center=<10.9192 -1.4411 -1.0468>, rotation=<rot 90 0 0>}]
 upperarm_r [{shapeType=Capsule, length=17.4826, radius=5.5477, center=<-10.9192 1.4411 1.0468>, rotation=<rot 90 0 180>}]
}
constraints [
 {bone1Name=spine_03, bone2Name=spine_01, swing1LimitDegrees=45, swing2LimitDegrees=45, twistLimitDegrees=45}
 {bone1Name=spine_04, bone2Name=spine_03, swing1LimitDegrees=45, swing2LimitDegrees=45, twistLimitDegrees=45}
 {bone1Name=spine_05, bone2Name=spine_04, swing1LimitDegrees=45, swing2LimitDegrees=45, twistLimitDegrees=45}
 {bone1Name=upperarm_l, bone2Name=spine_05, swing1LimitDegrees=45, swing2LimitDegrees=45, twistLimitDegrees=45}
 {bone1Name=upperarm_r, bone2Name=spine_05, swing1LimitDegrees=45, swing2LimitDegrees=45, twistLimitDegrees=45}
 {bone1Name=thigh_l, bone2Name=pelvis, swing1LimitDegrees=45, swing2LimitDegrees=45, twistLimitDegrees=45}
 {bone1Name=thigh_r, bone2Name=pelvis, swing1LimitDegrees=45, swing2LimitDegrees=45, twistLimitDegrees=45}
 {bone1Name=dyn_jacket_c_02_l, bone2Name=dyn_jacket_c_01_l, swing1LimitDegrees=25, swing2LimitDegrees=15, twistLimitDegrees=0}
 {bone1Name=dyn_jacket_a_02_l, bone2Name=dyn_jacket_a_01_l, swing1LimitDegrees=15, swing2LimitDegrees=15, twistLimitDegrees=0}
 {bone1Name=dyn_jacket_b_02_l, bone2Name=dyn_jacket_b_01_l, swing1LimitDegrees=15, swing2LimitDegrees=15, twistLimitDegrees=0}
 {bone1Name=dyn_jacket_b_01_l, bone2Name=upperarm_l, swing1LimitDegrees=15, swing2LimitDegrees=15, twistLimitDegrees=0}
 {bone1Name=dyn_jacket_a_01_l, bone2Name=upperarm_l, swing1LimitDegrees=15, swing2LimitDegrees=15, twistLimitDegrees=0}
 {bone1Name=dyn_jacket_c_02_r, bone2Name=dyn_jacket_c_01_r, swing1LimitDegrees=25, swing2LimitDegrees=25, twistLimitDegrees=0}
 {bone1Name=dyn_jacket_c_01_r, bone2Name=spine_04, swing1LimitDegrees=25, swing2LimitDegrees=25, twistLimitDegrees=0}
 {bone1Name=dyn_jacket_c_01_l, bone2Name=spine_04, swing1LimitDegrees=25, swing2LimitDegrees=15, twistLimitDegrees=0}
 {bone1Name=dyn_jacket_a_02_r, bone2Name=dyn_jacket_a_01_r, swing1LimitDegrees=15, swing2LimitDegrees=15, twistLimitDegrees=0}
 {bone1Name=dyn_jacket_a_01_r, bone2Name=upperarm_r, swing1LimitDegrees=15, swing2LimitDegrees=15, twistLimitDegrees=0}
 {bone1Name=dyn_jacket_b_02_r, bone2Name=dyn_jacket_b_01_r, swing1LimitDegrees=15, swing2LimitDegrees=15, twistLimitDegrees=0}
 {bone1Name=dyn_jacket_b_01_r, bone2Name=upperarm_r, swing1LimitDegrees=15, swing2LimitDegrees=15, twistLimitDegrees=0}
 {
  bone1Name=spine_01
  bone2Name=pelvis
  swing1LimitDegrees=45
  swing1Motion=Free
  swing2LimitDegrees=45
  swing2Motion=Free
  twistLimitDegrees=45
  twistMotion=Free
 }
 {bone1Name=dyn_jacket_d_02_r, bone2Name=dyn_jacket_d_01_r, swing1LimitDegrees=20, swing2LimitDegrees=60, twistLimitDegrees=5}
 {bone1Name=dyn_jacket_d_01_r, bone2Name=spine_03, swing1LimitDegrees=5, swing2LimitDegrees=15, twistLimitDegrees=15}
 {bone1Name=dyn_jacket_d_01_l, bone2Name=spine_03, swing1LimitDegrees=5, swing2LimitDegrees=15, twistLimitDegrees=15}
 {bone1Name=dyn_jacket_d_02_l, bone2Name=dyn_jacket_d_01_l, swing1LimitDegrees=20, swing2LimitDegrees=25, twistLimitDegrees=3}
 {
  bone1Name=dyn_chain_fr_02
  bone2Name=thigh_r
  swing1LimitDegrees=45
  swing1Motion=Locked
  swing2LimitDegrees=45
  swing2Motion=Locked
  twistLimitDegrees=45
  twistMotion=Locked
 }
]
