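# Control Rig for the ElegantHeist female face accessory that poses the back hair bones as the head and arms move.
# DeltaRotation turns head rotation relative to a null into weights that, with spherical pose readers, offset a hair aim target.
# The Backpack branch swaps arm-driven offsets for per-strand aim constraints and head-driven offsets, including a crouch adjustment.
# A later sequence step aims the hair base at that target and adds weighted corrective rotations.
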ControlRigBlueprint /BRCosmetics/Characters/Player/Female/Medium/Bodies/F_MED_ElegantHeist/Meshes/Parts/F_MED_ElegantHeist_FaceAcc_CtrlRig
graphs [
 New Function_ContainedGraph {nodes [
   n1 = RigVMFunctionEntryNode()
   RigVMFunction_MathBoolToInteger = &RigVMFunction_MathBoolToInteger(Value=@n1.onlyTwist)
   RigVMFunction_MathQuaternionSwingTwist = &RigVMFunction_MathQuaternionSwingTwist(
    TwistAxis=@n1.twistAxis
    Input=&RigUnit_GetRelativeTransformForItem(
     Child=(Type=Bone,Name="None")
     Parent=(Type=Bone,Name="None")
     bChildInitial=False
     bParentInitial=False
     Child.Name=@n1.child
     Parent.Name=@n1.parent
     Parent.Type=@n1.parentType
    ).RelativeTransform.Rotation
   )
   RigVMFunction_MathQuaternionToEuler = &RigVMFunction_MathQuaternionToEuler(
    Value=RigVMDispatchNode(
     Values=((),())
     Index=@RigVMFunction_MathBoolToInteger.Result
     Values.0=@RigVMFunction_MathQuaternionSwingTwist.Swing
     Values.1=@RigVMFunction_MathQuaternionSwingTwist.Twist
    ) {template="DISPATCH_RigVMDispatch_SelectInt32(in Index,in Values,out Result)"}.Result
    RotationOrder=@n1.Rotation Order
   )
   > DISPATCH_RigVMDispatch_Print = RigVMDispatchNode(ScreenColor=(R=1,G=1,B=1,A=1), ScreenDuration=1.5, Enabled=@n1.debugRotation, Value=@RigVMFunction_MathQuaternionToEuler.Result) {template="DISPATCH_RigVMDispatch_Print(in Prefix,in Value,in Enabled,in ScreenDuration,in ScreenColor)"}
   Clamp = &RigVMFunction_MathIntClamp(Value=@n1.primAxis)
   DISPATCH_RigVMDispatch_SelectInt32 = RigVMDispatchNode(
    Values=(0,0,0)
    Index=@Clamp.Result
    Values.0=@RigVMFunction_MathQuaternionToEuler.Result.X
    Values.1=@RigVMFunction_MathQuaternionToEuler.Result.Y
    Values.2=@RigVMFunction_MathQuaternionToEuler.Result.Z
   ) {template="DISPATCH_RigVMDispatch_SelectInt32(in Index,in Values,out Result)"}
   Remap = &RigVMFunction_MathDoubleRemap(
    TargetMaximum=1
    TargetMinimum=0
    bClamp=False
    Value=@DISPATCH_RigVMDispatch_SelectInt32.Result
    SourceMaximum=@n1.max
    SourceMinimum=@n1.min
   )
   Multiply = &RigVMFunction_MathDoubleMul(A=&RigVMFunction_MathDoubleClamp(Value=@Remap.Result).Result, B=@n1.premult)
   > RigVMFunctionReturnNode(outputWeight=@Multiply.Result)
  ]}
 RigVMFunctionLibrary {nodes [DeltaRotation = RigVMCollapseNode(ExecuteContext=(), Rotation Order=ZYX, premult=1, twistAxis=(X=1,Y=0,Z=0))]}
 RigVMFunction_Sequence_1_ContainedGraph {nodes [
   n1 = RigVMFunctionEntryNode()
   RigVMFunction_Sequence = &RigVMFunction_Sequence(ExecuteContext=@n1.ExecuteContext)
   n2 = &RigVMFunction_Sequence(ExecuteContext=@RigVMFunction_Sequence.B)
   n3 = &RigVMFunction_Sequence(ExecuteContext=@n2.B)
   n4 = &RigVMFunction_Sequence(ExecuteContext=@n3.B)
   RigVMFunctionReturnNode(A=@RigVMFunction_Sequence.A, B=@n2.A, C=@n3.A, D=@n4.A, E=@n4.B)
  ]}
 RigVMModel {nodes [
   RigUnit_BeginExecution = &RigUnit_BeginExecution()
   RigVMFunction_Sequence = RigVMAggregateNode(ExecuteContext=@RigUnit_BeginExecution.ExecutePin)
   DeltaRotation_0 = RigVMFunctionReferenceNode(
    Rotation Order=ZYX
    child=head
    max=-80
    parent=head_null
    parentType=Null
    premult=1
    primAxis=2
    twistAxis=(X=1,Y=0,Z=0)
    ExecuteContext=@RigVMFunction_Sequence.B
   )
   RigVMFunction_AnimEvalRichCurve = &RigVMFunction_AnimEvalRichCurve(
    Curve=(EditorCurveData=(Keys=((InterpMode=RCIM_Cubic),(InterpMode=RCIM_Cubic,TangentMode=RCTM_User,Time=0.3692,Value=0.5547,ArriveTangent=1.5332,LeaveTangent=1.5332),(Time=1,Value=1))))
    SourceMinimum=0
    TargetMinimum=0
    Value=@DeltaRotation_0.outputWeight
   )
   > n1 = &RigUnit_ModifyTransforms(
    ItemToModify=((Item=(Type=Bone,Name="base_hair_back_aimtarget"),Transform=(Rotation=(X=0,Y=0,Z=0,W=1),Translation=(X=0.0031,Y=6.3998,Z=-5.681),Scale3D=(X=1,Y=1,Z=1))))
    WeightMinimum=0
    Weight=@RigVMFunction_AnimEvalRichCurve.Result
   )
   RigVMFunction_ControlFlowBranch = &RigVMFunction_ControlFlowBranch(Condition=RigVMVariableNode() {variable=Backpack}.Value, ExecuteContext=@n1.ExecutePin)
   DeltaRotation_1 = RigVMFunctionReferenceNode(
    Rotation Order=ZYX
    child=head
    max=-40
    parent=head_null
    parentType=Null
    premult=1
    primAxis=1
    twistAxis=(X=1,Y=0,Z=0)
    ExecuteContext=@RigVMFunction_ControlFlowBranch.False
   )
   > &RigUnit_ModifyTransforms(
    ItemToModify=((Item=(Type=Bone,Name="base_hair_back_aimtarget"),Transform=(Rotation=(X=0,Y=0,Z=0,W=1),Translation=(X=-13,Y=3,Z=0),Scale3D=(X=1,Y=1,Z=1))))
    WeightMinimum=0
    Weight=@DeltaRotation_1.outputWeight
   )
   n2 = /Script/RigVM.RigVMFunction_MathBoolMake(Value=false) {is=struct RigVMUnitNode}
   > n3 = &RigUnit_SphericalPoseReader(
    ActiveRegionScaleFactors=(PositiveWidth=0.75,NegativeWidth=0.75,PositiveHeight=1,NegativeHeight=1)
    ActiveRegionSize=0.1
    Debug=(bDrawDebug=true,bDraw2D=false,DebugScale=25,DebugSegments=20,DebugThickness=0.25)
    DriverAxis=(X=0,Y=1,Z=0)
    DriverItem=(Type=Bone,Name="head")
    FalloffRegionScaleFactors=(PositiveWidth=1,NegativeWidth=1,PositiveHeight=1,NegativeHeight=2.2)
    FalloffSize=0.2
    OptionalParentItem=(Type=Bone)
    RotationOffset=(X=150,Y=0,Z=0)
    Debug.bDrawDebug=@n2.Value
   )
   > &RigUnit_ModifyTransforms(
    ItemToModify=((Item=(Type=Bone,Name="base_hair_back_aimtarget"),Transform=(Rotation=(X=0,Y=0,Z=0,W=1),Translation=(X=-12,Y=0,Z=0),Scale3D=(X=1,Y=1,Z=1))))
    WeightMinimum=0
    Weight=@n3.OutputParam
   )
   > n4 = &RigUnit_SphericalPoseReader(
    ActiveRegionScaleFactors=(PositiveWidth=0.75,NegativeWidth=0.75,PositiveHeight=1,NegativeHeight=1)
    ActiveRegionSize=0.1
    Debug=(bDrawDebug=true,bDraw2D=false,DebugScale=25,DebugSegments=20,DebugThickness=0.25)
    DriverAxis=(X=0,Y=1,Z=0)
    DriverItem=(Type=Bone,Name="head")
    FalloffRegionScaleFactors=(PositiveWidth=1,NegativeWidth=1,PositiveHeight=1,NegativeHeight=2.2)
    FalloffSize=0.2
    FlipHeightScaling=true
    FlipWidthScaling=false
    OptionalParentItem=(Type=Bone)
    RotationOffset=(X=30,Y=0,Z=0)
    Debug.bDrawDebug=@n2.Value
   )
   > &RigUnit_ModifyTransforms(
    ItemToModify=((Item=(Type=Bone,Name="base_hair_back_aimtarget"),Transform=(Rotation=(X=0,Y=0,Z=0,W=1),Translation=(X=12,Y=0,Z=0),Scale3D=(X=1,Y=1,Z=1))))
    WeightMinimum=0
    Weight=@n4.OutputParam
   )
   n5 = &RigVMFunction_MathBoolMake(Value=false)
   > n6 = &RigUnit_SphericalPoseReader(
    ActiveRegionScaleFactors=(PositiveWidth=3,NegativeWidth=1,PositiveHeight=1.5,NegativeHeight=1.5)
    ActiveRegionSize=0.1
    Debug=(bDrawDebug=true,bDraw2D=false,DebugScale=25,DebugSegments=20,DebugThickness=0.25)
    DriverAxis=(X=1,Y=0,Z=0)
    DriverItem=(Type=Bone,Name="upperarm_l")
    FalloffRegionScaleFactors=(PositiveWidth=1,NegativeWidth=1,PositiveHeight=1,NegativeHeight=1)
    FalloffSize=0.4
    OptionalParentItem=(Type=Bone)
    RotationOffset=(X=90,Y=0,Z=180)
    Debug.bDrawDebug=@n5.Value
   )
   > &RigUnit_ModifyTransforms(
    ItemToModify=((Item=(Type=Bone,Name="base_hair_back_aimtarget"),Transform=(Rotation=(X=0,Y=0,Z=0,W=1),Translation=(X=0,Y=-16,Z=0),Scale3D=(X=1,Y=1,Z=1))))
    WeightMinimum=0
    Weight=@n6.OutputParam
   )
   > n7 = &RigUnit_SphericalPoseReader(
    ActiveRegionScaleFactors=(PositiveWidth=3,NegativeWidth=1,PositiveHeight=1.5,NegativeHeight=1.5)
    ActiveRegionSize=0.1
    Debug=(bDrawDebug=true,bDraw2D=false,DebugScale=25,DebugSegments=20,DebugThickness=0.25)
    DriverAxis=(X=-1,Y=0,Z=0)
    DriverItem=(Type=Bone,Name="upperarm_r")
    FalloffRegionScaleFactors=(PositiveWidth=1,NegativeWidth=1,PositiveHeight=1,NegativeHeight=1)
    FalloffSize=0.4
    FlipHeightScaling=true
    FlipWidthScaling=false
    OptionalParentItem=(Type=Bone)
    RotationOffset=(X=90,Y=0,Z=0)
    Debug.bDrawDebug=@n5.Value
   )
   > &RigUnit_ModifyTransforms(
    ItemToModify=((Item=(Type=Bone,Name="base_hair_back_aimtarget"),Transform=(Rotation=(X=0,Y=0,Z=0,W=1),Translation=(X=0,Y=-16,Z=0),Scale3D=(X=1,Y=1,Z=1))))
    WeightMinimum=0
    Weight=@n7.OutputParam
   )
   &RigUnit_AimConstraintLocalSpaceOffset(
    Child=(Type=Bone,Name="dyn_hair_back_l_01")
    Filter=(bX=True,bY=true,bZ=true)
    Parents=((Item=(Type=Bone,Name="dyn_hair_back_l_02_aimtarget"),Weight=1))
    UpAxis=(X=0,Y=-1,Z=0)
    Weight=1
    WorldUp=(Target=(X=0,Y=-1,Z=0),Kind=Location,Space=(Type=Bone,Name="dyn_hair_back_l_02_aimtarget"))
    bMaintainOffset=true
    ExecutePin=@RigVMFunction_ControlFlowBranch.True
   )
   > &RigUnit_AimConstraintLocalSpaceOffset(
    Child=(Type=Bone,Name="dyn_hair_back_m_01")
    Filter=(bX=True,bY=true,bZ=true)
    Parents=((Item=(Type=Bone,Name="dyn_hair_back_m_02_aimtarget"),Weight=1))
    UpAxis=(X=0,Y=-1,Z=0)
    Weight=1
    WorldUp=(Target=(X=0,Y=-1,Z=0),Kind=Location,Space=(Type=Bone,Name="dyn_hair_back_m_02_aimtarget"))
    bMaintainOffset=true
   )
   > &RigUnit_AimConstraintLocalSpaceOffset(
    Child=(Type=Bone,Name="dyn_hair_back_r_01")
    Filter=(bX=True,bY=true,bZ=true)
    Parents=((Item=(Type=Bone,Name="dyn_hair_back_r_02_aimtarget"),Weight=1))
    UpAxis=(X=0,Y=-1,Z=0)
    Weight=1
    WorldUp=(Target=(X=0,Y=-1,Z=0),Kind=Location,Space=(Type=Bone,Name="dyn_hair_back_r_02_aimtarget"))
    bMaintainOffset=true
   )
   n8 = &RigVMFunction_MathBoolMake(Value=false)
   > n9 = &RigUnit_SphericalPoseReader(
    ActiveRegionScaleFactors=(PositiveWidth=3.5,NegativeWidth=0.6,PositiveHeight=1,NegativeHeight=1)
    ActiveRegionSize=0.1
    Debug=(bDrawDebug=true,bDraw2D=False,bDrawLocalAxes=False,DebugScale=25,DebugSegments=20,DebugThickness=0.25)
    DriverAxis=(X=0,Y=1,Z=0)
    DriverItem=(Type=Bone,Name="head")
    FalloffRegionScaleFactors=(PositiveWidth=0.01,NegativeWidth=0.4,PositiveHeight=1,NegativeHeight=1)
    FalloffSize=0.2
    FlipHeightScaling=false
    FlipWidthScaling=false
    OptionalParentItem=(Type=Bone,Name="None")
    RotationOffset=(X=90,Y=0,Z=-35)
    Debug.bDrawDebug=@n8.Value
   )
   > &RigUnit_ModifyTransforms(
    ItemToModify="((Item=(Type=Bone,Name=\"base_hair_back_aimtarget\"),Transform=(Rotation=(X=0,Y=0,Z=0,W=1),Translation=(X=1.5,Y=-5,Z=0),Scale3D=(X=1,Y=1,Z=1))),(Item=(Type=Bone,Name=\"dyn_hair_back_l_02b\"),Transform=(Rotation=(X=0,Y=0,Z=0,W=1),Translation=(X=0,Y=-11,Z=0),Sca ...(truncated)"
    WeightMinimum=0
    Weight=@n9.OutputParam
   )
   RigVMFunction_MathBoolToFloat = &RigVMFunction_MathBoolToFloat(Value=RigVMVariableNode() {variable=IsCrouching}.Value)
   Multiply = &RigVMFunction_MathDoubleMul(A=@RigVMFunction_MathBoolToFloat.Result, B=@n9.OutputParam)
   > &RigUnit_ModifyTransforms(
    ItemToModify="((Item=(Type=Bone,Name=\"base_hair_back_aimtarget\"),Transform=(Rotation=(X=0,Y=0,Z=0,W=1),Translation=(X=5,Y=12,Z=0),Scale3D=(X=1,Y=1,Z=1))),(Item=(Type=Bone,Name=\"dyn_hair_back_l_02b\"),Transform=(Rotation=(X=0,Y=0,Z=0,W=1),Translation=(X=5,Y=0,Z=5),Scale3D ...(truncated)"
    WeightMinimum=0
    Weight=@Multiply.Result
   )
   > n10 = &RigUnit_SphericalPoseReader(
    ActiveRegionScaleFactors=(PositiveWidth=0.75,NegativeWidth=0.75,PositiveHeight=1,NegativeHeight=1)
    ActiveRegionSize=0.1
    Debug=(bDrawDebug=true,bDraw2D=false,DebugScale=25,DebugSegments=20,DebugThickness=0.25)
    DriverAxis=(X=0,Y=1,Z=0)
    DriverItem=(Type=Bone,Name="head")
    FalloffRegionScaleFactors=(PositiveWidth=1,NegativeWidth=1,PositiveHeight=1,NegativeHeight=2.2)
    FalloffSize=0.2
    OptionalParentItem=(Type=Bone)
    RotationOffset=(X=150,Y=0,Z=0)
    Debug.bDrawDebug=@n8.Value
   )
   > &RigUnit_ModifyTransforms(
    ItemToModify=((Item=(Type=Bone,Name="base_hair_back_aimtarget"),Transform=(Rotation=(X=0,Y=0,Z=0,W=1),Translation=(X=-12,Y=0,Z=0),Scale3D=(X=1,Y=1,Z=1))))
    WeightMinimum=0
    Weight=@n10.OutputParam
   )
   > n11 = &RigUnit_SphericalPoseReader(
    ActiveRegionScaleFactors=(PositiveWidth=0.75,NegativeWidth=0.75,PositiveHeight=1,NegativeHeight=1)
    ActiveRegionSize=0.1
    Debug=(bDrawDebug=true,bDraw2D=false,DebugScale=25,DebugSegments=20,DebugThickness=0.25)
    DriverAxis=(X=0,Y=1,Z=0)
    DriverItem=(Type=Bone,Name="head")
    FalloffRegionScaleFactors=(PositiveWidth=1,NegativeWidth=1,PositiveHeight=1,NegativeHeight=2.2)
    FalloffSize=0.2
    FlipHeightScaling=true
    FlipWidthScaling=false
    OptionalParentItem=(Type=Bone)
    RotationOffset=(X=55,Y=-20,Z=0)
    Debug.bDrawDebug=@n8.Value
   )
   > &RigUnit_ModifyTransforms(
    ItemToModify=((Item=(Type=Bone,Name="base_hair_back_aimtarget"),Transform=(Rotation=(X=0,Y=0,Z=0,W=1),Translation=(X=5,Y=0,Z=0),Scale3D=(X=1,Y=1,Z=1))))
    WeightMinimum=0
    Weight=@n11.OutputParam
   )
   &RigUnit_AimConstraintLocalSpaceOffset(
    Child=(Type=Bone,Name="base_hair_back")
    Filter=(bX=True,bY=true,bZ=true)
    Parents=((Item=(Type=Bone,Name="base_hair_back_aimtarget"),Weight=1))
    UpAxis=(X=0,Y=-1,Z=0)
    Weight=1
    WorldUp=(Target=(X=0,Y=-1,Z=0),Kind=Location,Space=(Type=Bone,Name="base_hair_back_aimtarget"))
    bMaintainOffset=true
    ExecutePin=@RigVMFunction_Sequence.C
   )
   > &RigUnit_ModifyTransforms(
    ItemToModify=((Item=(Type=Bone,Name="base_hair_back"),Transform=(Rotation=(X=0.0759,Y=0.0989,Z=-0.0387,W=0.9914),Translation=(X=0,Y=0,Z=0),Scale3D=(X=1,Y=1,Z=1))))
    WeightMinimum=0
    Weight=@DeltaRotation_1.outputWeight
   )
   > &RigUnit_ModifyTransforms(
    ItemToModify=((Item=(Type=Bone,Name="base_hair_back"),Transform=(Rotation=(X=0.173,Y=-0.0151,Z=-0.0858,W=0.9811),Translation=(X=0,Y=0,Z=0),Scale3D=(X=1,Y=1,Z=1))))
    WeightMinimum=0
    Weight=@RigVMFunction_AnimEvalRichCurve.Result
   )
   > &RigUnit_ModifyTransforms(
    ItemToModify=((Item=(Type=Bone,Name="base_hair_back"),Transform=(Rotation=(X=0.0474,Y=0.0722,Z=0.0115,W=0.9962),Translation=(X=0,Y=0,Z=0),Scale3D=(X=1,Y=1,Z=1))))
    WeightMinimum=0
    Weight=@n3.OutputParam
   )
   > &RigUnit_ModifyTransforms(
    ItemToModify=((Item=(Type=Bone,Name="base_hair_back"),Transform=(Rotation=(X=-0.0709,Y=-0.1078,Z=0.0196,W=0.9914),Translation=(X=0,Y=0,Z=0),Scale3D=(X=1,Y=1,Z=1))))
    WeightMinimum=0
    Weight=@n4.OutputParam
   )
   RigUnit_GetTransform = &RigUnit_GetTransform(Item=(Type=Control,Name="head_ctrl"), bInitial=False)
   RigUnit_SetTransform = &RigUnit_SetTransform(ExecutePin=(), Item=(Type=Bone,Name="head"), Weight=1, bInitial=False, Value=@RigUnit_GetTransform.Transform)
  ]}
]
hierarchy [
 root {loc=<0 0 0>, rot=<0 0 0 1>}
 pelvis {parent=root, loc=<0 -0.5301 98.0571>, rot=<0.0127 0.707 -0.0127 -0.707>}
 spine_01 {parent=pelvis, loc=<6.1208 -0.9328 0>, rot=<0 0 0.034 -0.9994>}
 spine_02 {parent=spine_01, loc=<6.5997 -0.2095 0>, rot=<0 0 0.0238 -0.9997>}
 spine_03 {parent=spine_02, loc=<6.4914 -0.1237 0>, rot=<0 0 0.0149 -0.9999>}
 spine_04 {parent=spine_03, loc=<6.9598 0.0794 0>, rot=<0 0 -0.0132 -0.9999>}
 spine_05 {parent=spine_04, loc=<6.9987 0 0>, rot=<0 0 -0.0015 -1>}
 neck_01 {parent=spine_05, loc=<11.3138 0.1529 0>, rot=<0 0 -0.2003 -0.9797>}
 neck_02 {parent=neck_01, loc=<4.5627 0 0>, rot=<0 0 0.0504 -0.9987>}
 head {parent=neck_02, loc=<5.3914 0 0>, rot=<0 0 0.115 -0.9934>}
 base_hair_back {parent=head, loc=<16.6486 -6.4972 -0.1508>, rot=<0.4335 -0.9011 0 0>}
 base_hair_back_l {parent=base_hair_back, loc=<7.8286 0.553 6.7091>, rot=<0 0 -0.43 -0.9028>}
 dyn_hair_back_l_01 {parent=base_hair_back_l, loc=<0 0 0>, rot=<0 0 0 1>}
 dyn_hair_back_l_02 {parent=dyn_hair_back_l_01, loc=<13.4489 0.0498 0>, rot=<0 0 0.2094 -0.9778>}
 dyn_hair_back_l_03 {parent=dyn_hair_back_l_02, loc=<11.1308 -0.0859 0>, rot=<0 0 -0.1373 -0.9905>}
 dyn_hair_back_l_04 {parent=dyn_hair_back_l_03, loc=<10.0891 0.0023 0>, rot=<0 0 0 -1>}
 dyn_hair_back_l_02b {parent=dyn_hair_back_l_02, loc=<2.8835 1.8226 5.8218>, rot=<0 0 0.0087 -1>}
 dyn_hair_back_l_03b {parent=dyn_hair_back_l_02b, loc=<14.0342 0.0116 0.0146>, rot=<0 0 0 -1>}
 base_hair_back_r {parent=base_hair_back, loc=<8.5201 0.5701 -8.0197>, rot=<0 0 -0.4062 -0.9138>}
 dyn_hair_back_r_01 {parent=base_hair_back_r, loc=<0 0 0>, rot=<0 0 0 1>}
 dyn_hair_back_r_02 {parent=dyn_hair_back_r_01, loc=<11.8864 -0.0262 0>, rot=<0 0 0.1254 -0.9921>}
 dyn_hair_back_r_03 {parent=dyn_hair_back_r_02, loc=<12.085 -0.0239 0>, rot=<0 0 -0.1194 -0.9929>}
 dyn_hair_back_r_04 {parent=dyn_hair_back_r_03, loc=<9.2532 0.0704 0>, rot=<0 0 0 -1>}
 dyn_hair_back_r_02b {parent=dyn_hair_back_r_02, loc=<4.9361 3.7632 -7.621>, rot=<0 0 0.0677 -0.9977>}
 dyn_hair_back_r_03b {parent=dyn_hair_back_r_02b, loc=<8.4203 -0.047 0.0249>, rot=<0 0 0 -1>}
 base_hair_back_m {parent=base_hair_back, loc=<8.4486 -0.0349 -0.3581>, rot=<0 0 -0.4182 -0.9084>}
 dyn_hair_back_m_01 {parent=base_hair_back_m, loc=<0 0 0>, rot=<0 0 0 1>}
 dyn_hair_back_m_02 {parent=dyn_hair_back_m_01, loc=<13.1016 0.0064 0.0003>, rot=<0 0 0.1863 -0.9825>}
 dyn_hair_back_m_03 {parent=dyn_hair_back_m_02, loc=<11.1018 -0.0171 0>, rot=<0 0 -0.1906 -0.9817>}
 dyn_hair_back_m_04 {parent=dyn_hair_back_m_03, loc=<9.8854 0.0621 0>, rot=<0 0 0 -1>}
 base_earing_l {parent=head, loc=<-0.316 0.331 -6.002>, rot=<0.004 -0.9497 0.3132 -0.0013>}
 dyn_earing_l {parent=base_earing_l, loc=<0 0 0>, rot=<0 0 0 1>}
 base_earing_r {parent=head, loc=<-0.316 0.3315 6.002>, rot=<0.0039 -0.9483 -0.3173 0.0013>}
 dyn_earing_r {parent=base_earing_r, loc=<0 0 0>, rot=<0 0 0 1>}
 base_hair_fr_01 {parent=head, loc=<8.3555 7.4029 3.4137>, rot=<-0.3898 0.1224 -0.4021 -0.8194>}
 dyn_hair_fr_01 {parent=base_hair_fr_01, loc=<0 0 0>, rot=<0 0 0 1>}
 dyn_hair_fr_02 {parent=dyn_hair_fr_01, loc=<2.9327 0.0088 -0.0042>, rot=<-0.0752 0.0264 -0.8306 -0.5511>}
 dyn_hair_fr_03 {parent=dyn_hair_fr_02, loc=<6.847 0.0092 0.0108>, rot=<0.0163 -0.1312 -0.286 -0.9491>}
 dyn_hair_fr_04 {parent=dyn_hair_fr_03, loc=<6.746 0.0125 -0.0314>, rot=<0.0838 0.3295 0.3473 -0.8739>}
 dyn_hair_fr_05 {parent=dyn_hair_fr_04, loc=<5.0917 -0.0391 0.0212>, rot=<0.0673 0.03 0.9189 0.3875>}
 FACIAL_C_FacialRoot {parent=head, loc=<1.826 0.6057 0>, rot=<0.5021 -0.4979 -0.5021 -0.4979>}
 FACIAL_C_Forehead {parent=FACIAL_C_FacialRoot, loc=<0.0001 -2.428 7.4971>, rot=<-0.0464 0 0 -0.9989>}
 FACIAL_L_ForeheadIn {parent=FACIAL_C_FacialRoot, loc=<1.723 -2.9403 7.3899>, rot=<-0.0611 -0.0666 0.0062 -0.9959>}
 FACIAL_R_ForeheadIn {parent=FACIAL_C_FacialRoot, loc=<-1.6965 -2.9979 7.3731>, rot=<-0.061 0.0669 -0.0059 -0.9959>}
 FACIAL_L_ForeheadMid {parent=FACIAL_C_FacialRoot, loc=<3.1191 -3.166 6.9051>, rot=<-0.0713 -0.1748 -0.0281 -0.9816>}
 FACIAL_R_ForeheadMid {parent=FACIAL_C_FacialRoot, loc=<-3.0818 -3.2172 6.9051>, rot=<-0.0714 0.1749 0.0268 -0.9816>}
 FACIAL_L_ForeheadOut {parent=FACIAL_C_FacialRoot, loc=<4.5575 -3.0271 6.2029>, rot=<-0.1242 -0.2561 -0.0842 -0.9549>}
 FACIAL_R_ForeheadOut {parent=FACIAL_C_FacialRoot, loc=<-4.5064 -3.102 6.1995>, rot=<-0.1233 0.2563 0.0829 -0.9551>}
 FACIAL_L_EyesackUpper {parent=FACIAL_C_FacialRoot, loc=<3.3129 -1.9229 7.7458>, rot=<0.0564 -0.083 -0.0559 -0.9934>}
 FACIAL_R_EyesackUpper {parent=FACIAL_C_FacialRoot, loc=<-3.2892 -1.9678 7.7933>, rot=<0.0491 0.0854 0.0627 -0.9932>}
 FACIAL_R_EyelidUpperFurrow {parent=FACIAL_C_FacialRoot, loc=<-3.1585 -2.2823 7.7526>, rot=<-0.0305 0.1506 0.0786 -0.985>}
 FACIAL_L_EyelidUpperA {parent=FACIAL_C_FacialRoot, loc=<2.9017 -1.4759 5.7628>, rot=<0.2941 0.0704 -0.0739 -0.9503>}
 FACIAL_R_Eye {parent=FACIAL_C_FacialRoot, loc=<-2.9046 -1.5642 5.626>, rot=<0 -0.153 0 -0.9882>}
 FACIAL_R_EyelidLowerA {parent=FACIAL_C_FacialRoot, loc=<-2.926 -1.5642 5.626>, rot=<0.2804 -0.0061 -0.008 -0.9598>}
 FACIAL_L_EyeCornerOuter {parent=FACIAL_C_FacialRoot, loc=<4.2749 -1.1069 7.0297>, rot=<0.0032 -0.3744 0.0043 -0.9272>}
 FACIAL_R_EyeCornerOuter {parent=FACIAL_C_FacialRoot, loc=<-4.5776 -1.4332 6.5295>, rot=<0 0.3827 0 -0.9239>}
 FACIAL_R_EyesackLower {parent=FACIAL_C_FacialRoot, loc=<-2.8796 0.7359 6.0778>, rot=<-0.0217 0.135 0.0231 -0.9903>}
 FACIAL_R_CheekInner {parent=FACIAL_C_FacialRoot, loc=<-2.1405 0.9497 6.5376>, rot=<-0.1243 0.09 0.1975 -0.9682>}
 FACIAL_R_CheekOuter {parent=FACIAL_C_FacialRoot, loc=<-4.1159 0.9658 5.1361>, rot=<0.0565 0.3762 -0.1 -0.9194>}
 FACIAL_R_NasolabialBulge {parent=FACIAL_C_FacialRoot, loc=<-3.6531 4.3167 5.1723>, rot=<0.0597 0.3706 -0.0356 -0.9262>}
 FACIAL_R_NasolabialFurrow {parent=FACIAL_C_FacialRoot, loc=<-2.812 3.0889 7.0311>, rot=<-0.0329 0.1267 0.3635 -0.9223>}
 FACIAL_R_CheekLower {parent=FACIAL_C_FacialRoot, loc=<-4.3079 3.9549 4.5149>, rot=<0.0365 0.4272 -0.0612 -0.9013>}
 FACIAL_C_Nose {parent=FACIAL_C_FacialRoot, loc=<0 0.7973 7.7667>, rot=<-0.1305 0 0.0003 -0.9914>}
 FACIAL_C_MouthUpper {parent=FACIAL_C_FacialRoot, loc=<0 4.4708 6.0634>, rot=<0 0 0 -1>}
 FACIAL_R_LipCorner {parent=FACIAL_C_MouthUpper, loc=<-2.2081 0.6251 0.323>, rot=<0.014 0.2112 0.0195 -0.9772>}
 FACIAL_L_Masseter {parent=FACIAL_C_FacialRoot, loc=<6.2463 1.2804 3.0238>, rot=<0.0703 0.1129 0.0071 -0.9911>}
 FACIAL_R_Masseter {parent=FACIAL_C_FacialRoot, loc=<-6.2434 1.2768 3.0245>, rot=<0.0701 -0.1136 -0.0078 -0.991>}
 FACIAL_L_UnderChin {parent=FACIAL_C_FacialRoot, loc=<2.8095 7.1968 1.8843>, rot=<0.401 -0.3539 0.0837 -0.8408>}
 FACIAL_R_UnderChin {parent=FACIAL_C_FacialRoot, loc=<-2.8499 7.1717 1.9067>, rot=<0.4012 0.354 -0.0832 -0.8407>}
 FACIAL_C_LowerLipRotation {parent=FACIAL_C_FacialRoot, loc=<0 2.0037 1.1413>, rot=<0.342 0 0 -0.9397>}
 FACIAL_C_Jaw {parent=FACIAL_C_FacialRoot, loc=<0 2.0037 1.1413>, rot=<0.342 0 0 -0.9397>}
 FACIAL_R_ChinSide {parent=FACIAL_C_Jaw, loc=<-2.8423 0.6386 6.6488>, rot=<-0.2129 0.3733 -0.0477 -0.9017>}
 FACIAL_C_Jawline {parent=FACIAL_C_Jaw, loc=<0.0046 2.6983 6.984>, rot=<0.2557 0.0004 -0.0002 -0.9667>}
 FACIAL_L_Ear {parent=FACIAL_C_FacialRoot, loc=<6.3224 -0.2425 0.1892>, rot=<-0.0262 -0.0007 -0.0262 -0.9993>}
 FACIAL_R_Ear {parent=FACIAL_C_FacialRoot, loc=<-6.3224 -0.2423 0.1892>, rot=<-0.0262 0.0007 0.0262 -0.9993>}
 FACIAL_L_Jawline {parent=FACIAL_C_FacialRoot, loc=<4.1213 5.5029 3.1428>, rot=<0.2059 0.1576 -0.0552 -0.9642>}
 FACIAL_R_Jawline {parent=FACIAL_C_FacialRoot, loc=<-4.1191 5.4981 3.1536>, rot=<0.2073 -0.1577 0.055 -0.9639>}
 FACIAL_C_UnderChin {parent=FACIAL_C_FacialRoot, loc=<-0.0068 8.5365 2.9634>, rot=<0.4991 -0.0001 -0.0001 -0.8666>}
 base_hair_back_aimtarget {parent=spine_05, loc=<8.277 -13.425 0>, rot=<0.0284 -0.7065 -0.0284 -0.7065>}
 head_null {parent=spine_05, type=Null, loc=<20.6573 3.5489 0>, rot=<0 0 -0.036 -0.9994>}
 attach_cape {parent=spine_05, loc=<9.016 -11.7106 -0.0013>, rot=<0.0616 -0.7044 -0.0616 -0.7044>}
 attach_backpack {parent=spine_05, loc=<-0.1213 -13.469 -0.0008>, rot=<-0.0093 -0.707 0.0093 -0.707>}
 clavicle_l {parent=spine_05, loc=<5.4831 -0.1053 -3.2946>, rot=<0.0465 -0.7289 -0.009 -0.6829>}
 upperarm_l {parent=clavicle_l, loc=<9.8658 -1.2136 -0.4105>, rot=<0.0164 -0.3245 -0.0269 -0.9454>}
 upperarm_twist_01_l {parent=upperarm_l, loc=<7.2127 -0.1129 0.0005>, rot=<0 0 0 -1>}
 deform_upperarm_bk_base_l {parent=upperarm_twist_01_l, loc=<-6.4047 -4.375 0.3136>, rot=<0 0 0 -1>}
 deform_upperarm_bk_l {parent=deform_upperarm_bk_base_l, loc=<0 0 0>, rot=<0 0 0 1>}
 deform_upperarm_up_base_l {parent=upperarm_twist_01_l, loc=<-8.8907 0.4823 3.9514>, rot=<0 0 0 -1>}
 deform_upperarm_up_l {parent=deform_upperarm_up_base_l, loc=<0 0 0>, rot=<0 0 0 1>}
 base_jacket_a_01_l {parent=upperarm_l, loc=<3.3401 -0.4311 4.7453>, rot=<-0.0192 0.3207 0.0404 -0.9461>}
 dyn_jacket_a_01_l {parent=base_jacket_a_01_l, loc=<0 0 0>, rot=<0 0 0 1>}
 dyn_jacket_a_02_l {parent=dyn_jacket_a_01_l, loc=<3.4111 -0.0085 1.5198>, rot=<0 0 0 1>}
 base_jacket_b_01_l {parent=upperarm_l, loc=<5.8618 4.6155 0.1836>, rot=<0.1893 0.2994 -0.5111 -0.7831>}
 dyn_jacket_b_01_l {parent=base_jacket_b_01_l, loc=<0 0 0>, rot=<0 0 0 1>}
 dyn_jacket_b_02_l {parent=dyn_jacket_b_01_l, loc=<3.6777 -0.0004 -0.0001>, rot=<0 0 0 1>}
 clavicle_r {parent=spine_05, loc=<5.4831 -0.1053 3.2946>, rot=<0.7289 0.0465 -0.6829 0.009>}
 upperarm_r {parent=clavicle_r, loc=<-9.8658 1.2136 0.4105>, rot=<0.0164 -0.3245 -0.0269 -0.9454>}
 upperarm_twist_01_r {parent=upperarm_r, loc=<-7.2128 0.1131 0.0002>, rot=<0 0 0 -1>}
 deform_upperarm_bk_base_r {parent=upperarm_twist_01_r, loc=<6.4048 4.3747 -0.3142>, rot=<0 0 0 -1>}
 deform_upperarm_bk_r {parent=deform_upperarm_bk_base_r, loc=<0 0 0>, rot=<0 0 0 1>}
 deform_upperarm_up_base_r {parent=upperarm_twist_01_r, loc=<8.8907 -0.4825 -3.9518>, rot=<0 0 0 -1>}
 deform_upperarm_up_r {parent=deform_upperarm_up_base_r, loc=<0 0 0>, rot=<0 0 0 1>}
 base_jacket_a_01_r {parent=upperarm_r, loc=<-3.3395 0.4316 -4.7458>, rot=<-0.3207 -0.0192 0.9461 0.0404>}
 dyn_jacket_a_01_r {parent=base_jacket_a_01_r, loc=<0 0 0>, rot=<0 0 0 1>}
 dyn_jacket_a_02_r {parent=dyn_jacket_a_01_r, loc=<3.3403 -0.0088 -1.5202>, rot=<0 0 0 1>}
 base_jacket_b_01_r {parent=upperarm_r, loc=<-5.8615 -4.6159 -0.1827>, rot=<-0.2994 0.1893 0.7831 -0.5111>}
 dyn_jacket_b_01_r {parent=base_jacket_b_01_r, loc=<0 0 0>, rot=<0 0 0 1>}
 dyn_jacket_b_02_r {parent=dyn_jacket_b_01_r, loc=<3.6778 0 0.0001>, rot=<0 0 0 1>}
 dyn_hair_back_r_02_aimtarget {parent=spine_05, loc=<10.4027 -8.9698 10.0763>, rot=<-0.1949 -0.9668 0.1152 -0.1183>}
 dyn_hair_back_m_02_aimtarget {parent=spine_05, loc=<11.1074 -10.6857 2.7269>, rot=<-0.1476 -0.9749 0.1214 -0.1145>}
 dyn_hair_back_l_02_aimtarget {parent=spine_05, loc=<12.7342 -11.4301 -4.084>, rot=<-0.1387 -0.9759 0.1232 -0.1151>}
 thigh_r {parent=pelvis, loc=<-1.0852 -0.2106 8.7051>, rot=<0.0355 -0.0009 0.9993 0.0155>}
 base_chain_acc_cn_01 {parent=thigh_r, loc=<17.0957 -0.5101 11.646>, rot=<0.9665 0.0235 0.133 -0.2182>}
 base_hair_back_ctrl {controlType=EULER_TRANSFORM, type=Control, loc=<0 0 0>, rot=<0 0 0 1>}
 head_ctrl {controlType=EULER_TRANSFORM, type=Control, loc=<0 0 0>, rot=<0 0 0 1>}
]
variables [
 Backpack {type=bool, default=False}
 IsCrouching {type=bool, default=False}
 UpButCrouching {type=double, default=0}
]
