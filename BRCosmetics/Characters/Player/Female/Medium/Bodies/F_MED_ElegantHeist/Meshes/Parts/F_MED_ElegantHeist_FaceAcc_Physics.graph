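# Physics asset for the ElegantHeist female medium FaceAcc cosmetic mesh, setting up secondary motion for hair and earrings.
# Capsules on upper spine, shoulders, neck and head form body volumes around the dynamic chains.
# Earrings, a front strand and three back strands, the side ones branching, hang from base bones with tight twist and swing generally loosening toward tips.
# Angularly free cross-constraints link both side strands to the middle one at the third segment.
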PhysicsAsset /BRCosmetics/Characters/Player/Female/Medium/Bodies/F_MED_ElegantHeist/Meshes/Parts/F_MED_ElegantHeist_FaceAcc_Physics
bodies {
 base_earing_l [{shapeType=Capsule, length=0.505, radius=0.505, shapeName=base_earing_l_capsule, rotation=<rot 0 -90 0>}]
 base_earing_r [{shapeType=Capsule, length=0.505, radius=0.505, shapeName=base_earing_l_capsule, rotation=<rot 0 -90 0>}]
 base_hair_back [{shapeType=Capsule, length=0.505, radius=0.505, shapeName=base_hair_back_capsule, rotation=<rot 0 -90 0>}]
 base_hair_back_l [{shapeType=Capsule, length=0.5, radius=0.5, shapeName=base_hair_back_l_capsule}]
 base_hair_back_m [{shapeType=Capsule, length=0.505, radius=0.505, shapeName=base_hair_back_m_capsule, rotation=<rot 0 -90 0>}]
 base_hair_back_r [{shapeType=Capsule, length=0.505, radius=0.505, shapeName=base_hair_back_r_capsule, rotation=<rot 0 -90 0>}]
 base_hair_fr_01 [{shapeType=Capsule, length=0.5, radius=0.5, shapeName=base_hair_fr_01_capsule}]
 clavicle_l [{shapeType=Capsule, length=6, radius=9, shapeName=clavicle_l_capsule, center=<7.3 -2 -4.2>, rotation=<rot 90 0 -10>}]
 clavicle_r [{shapeType=Capsule, length=6, radius=9, shapeName=clavicle_l_capsule, center=<-7.3 2 3.2018>, rotation=<rot 90 165.9638 155.9638>}]
 dyn_earing_l [{shapeType=Capsule, length=1, radius=1.3, shapeName=base_earing_r_capsule, center=<4 0 0>, rotation=<rot 90 0 0>}]
 dyn_earing_r [{shapeType=Capsule, length=1, radius=1.3, shapeName=dyn_earing_r_capsule, center=<4 0 0>, rotation=<rot 90 0 0>}]
 dyn_hair_back_l_01 [{shapeType=Capsule, length=10, radius=4, shapeName=dyn_hair_back_l_01_capsule, center=<5.5 0 2>, rotation=<rot -70 0 0>}]
 dyn_hair_back_l_02 [{shapeType=Capsule, length=9, radius=4, shapeName=dyn_hair_back_l_02_capsule, center=<3 1.8 0>, rotation=<rot 90 0 0>}]
 dyn_hair_back_l_02b [{shapeType=Capsule, length=9, radius=4, shapeName=dyn_hair_back_l_02b_capsule, center=<1.6 1 0>, rotation=<rot 90 -172.875 -162.875>}]
 dyn_hair_back_l_03 [{shapeType=Capsule, length=10, radius=4, shapeName=dyn_hair_back_l_03_capsule, center=<3.2 0.5 0>, rotation=<rot 90 -90 -80>}]
 dyn_hair_back_l_03b [{shapeType=Capsule, length=10, radius=4, shapeName=dyn_hair_back_l_03b_capsule, center=<-0.25 1.5 3.2>, rotation=<rot 90 164.8459 154.8459>}]
 dyn_hair_back_l_04 [{shapeType=Capsule, length=16, radius=4, shapeName=dyn_hair_back_l_04_capsule, center=<5 0 2>, rotation=<rot 45 0 0>}]
 dyn_hair_back_m_01 [{shapeType=Capsule, length=10, radius=4, shapeName=dyn_hair_back_m_01_capsule, center=<6.3 0 0>, rotation=<rot 90 0 0>}]
 dyn_hair_back_m_02 [{shapeType=Capsule, length=9, radius=4, shapeName=dyn_hair_back_m_02_capsule, center=<3 1.3 0>, rotation=<rot 90 0 0>}]
 dyn_hair_back_m_03 [{shapeType=Capsule, length=10, radius=4, shapeName=dyn_hair_back_m_03_capsule, center=<3 -0.75 0>, rotation=<rot 90 177.6141 -162.3859>}]
 dyn_hair_back_m_04 [{shapeType=Capsule, length=16, radius=4, shapeName=dyn_hair_back_m_04_capsule, center=<3.6 -0.8 0>, rotation=<rot 90 0 0>}]
 dyn_hair_back_r_01 [{shapeType=Capsule, length=10, radius=4, shapeName=dyn_hair_back_r_01_capsule, center=<4 0 -3.6>, rotation=<rot 70 0 0>}]
 dyn_hair_back_r_02 [{shapeType=Capsule, length=9, radius=4, shapeName=dyn_hair_back_r_02_capsule, center=<3 1.8 0>, rotation=<rot 90 0 0>}]
 dyn_hair_back_r_02b [{shapeType=Capsule, length=14, radius=4, shapeName=dyn_hair_back_r_02b_capsule, center=<0 0.4 -0.3>, rotation=<rot 90 -165.9638 -155.9638>}]
 dyn_hair_back_r_03 [{shapeType=Capsule, length=10, radius=4, shapeName=dyn_hair_back_r_03_capsule, center=<3 -0.5 0>, rotation=<rot 90 177.6141 -162.3859>}]
 dyn_hair_back_r_03b [{shapeType=Capsule, length=10, radius=4, shapeName=dyn_hair_back_r_03b_capsule, center=<0 0 -1.8>, rotation=<rot 90 0 0>}]
 dyn_hair_back_r_04 [{shapeType=Capsule, length=16, radius=4, shapeName=dyn_hair_back_r_04_capsule, center=<4.5 1.1 -4.3>, rotation=<rot -35 0 0>}]
 dyn_hair_fr_01 [{shapeType=Capsule, length=2, radius=2, shapeName=dyn_hair_fr_01_capsule, rotation=<rot 0 0 90>}]
 dyn_hair_fr_02 [{shapeType=Capsule, length=3, radius=3, shapeName=dyn_hair_fr_02_capsule, center=<7.5 0 0>}]
 dyn_hair_fr_03 [{shapeType=Capsule, length=4, radius=3.3, shapeName=dyn_hair_fr_03_capsule, center=<5 -1.55 0.6>}]
 dyn_hair_fr_04 [{shapeType=Capsule, length=4, radius=2.2, shapeName=dyn_hair_fr_04_capsule, center=<1.5 -0.7 0.15>, rotation=<rot 90 0 0>}]
 dyn_hair_fr_05 [{shapeType=Capsule, length=3, radius=1.2, shapeName=dyn_hair_fr_05_capsule, center=<1.7 -1 0>, rotation=<rot 0 -90 -90>}]
 head [{shapeType=Capsule, length=11, radius=6.2, shapeName=head_capsule, center=<3.7 2 0>, rotation=<rot 0 60 90>}]
 neck_01 [{shapeType=Capsule, length=2, radius=4.5, shapeName=neck_01_capsule, rotation=<rot 90 0 0>}]
 neck_02 [{shapeType=Capsule, length=2, radius=4.5, shapeName=neck_02_capsule, rotation=<rot 90 0 0>}]
 spine_03 [{shapeType=Capsule, length=22, radius=11.2, shapeName=spine_04_capsule, center=<-12.5 -0.7 0>, rotation=<rot 90 0 -15>}]
 spine_04 [{shapeType=Capsule, length=11, radius=9.8, shapeName=spine_04_capsule, center=<-4 -2.2 0>}]
 spine_05 [{shapeType=Capsule, length=11, radius=11, shapeName=spine_05_capsule, center=<0.1518 -1 0>}]
 upperarm_l [{shapeType=Capsule, length=18, radius=8.5, shapeName=upperarm_l_capsule, center=<16 0 0.8>, rotation=<rot 90 0 0>}]
 upperarm_r [{shapeType=Capsule, length=18, radius=8.5, shapeName=upperarm_l_capsule, center=<-16 0 -0.8>, rotation=<rot 90 0 0>}]
}
constraints [
 {
  bone1Name=spine_04
  bone2Name=spine_03
  swing1LimitDegrees=45
  swing1Motion=Free
  swing2LimitDegrees=45
  swing2Motion=Free
  twistLimitDegrees=45
  twistMotion=Free
 }
 {
  bone1Name=spine_05
  bone2Name=spine_04
  swing1LimitDegrees=45
  swing1Motion=Free
  swing2LimitDegrees=45
  swing2Motion=Free
  twistLimitDegrees=45
  twistMotion=Free
 }
 {
  bone1Name=clavicle_l
  bone2Name=spine_05
  swing1LimitDegrees=45
  swing1Motion=Free
  swing2LimitDegrees=45
  swing2Motion=Free
  twistLimitDegrees=45
  twistMotion=Free
 }
 {
  bone1Name=upperarm_l
  bone2Name=clavicle_l
  swing1LimitDegrees=45
  swing1Motion=Free
  swing2LimitDegrees=45
  swing2Motion=Free
  twistLimitDegrees=45
  twistMotion=Free
 }
 {
  bone1Name=clavicle_r
  bone2Name=spine_05
  swing1LimitDegrees=45
  swing1Motion=Free
  swing2LimitDegrees=45
  swing2Motion=Free
  twistLimitDegrees=45
  twistMotion=Free
 }
 {
  bone1Name=upperarm_r
  bone2Name=clavicle_r
  swing1LimitDegrees=45
  swing1Motion=Free
  swing2LimitDegrees=45
  swing2Motion=Free
  twistLimitDegrees=45
  twistMotion=Free
 }
 {
  bone1Name=neck_01
  bone2Name=spine_05
  swing1LimitDegrees=45
  swing1Motion=Free
  swing2LimitDegrees=45
  swing2Motion=Free
  twistLimitDegrees=45
  twistMotion=Free
 }
 {
  bone1Name=neck_02
  bone2Name=neck_01
  swing1LimitDegrees=45
  swing1Motion=Free
  swing2LimitDegrees=45
  swing2Motion=Free
  twistLimitDegrees=45
  twistMotion=Free
 }
 {
  bone1Name=head
  bone2Name=neck_02
  swing1LimitDegrees=45
  swing1Motion=Free
  swing2LimitDegrees=45
  swing2Motion=Free
  twistLimitDegrees=45
  twistMotion=Free
 }
 {
  bone1Name=base_earing_l
  bone2Name=head
  swing1LimitDegrees=45
  swing1Motion=Free
  swing2LimitDegrees=45
  swing2Motion=Free
  twistLimitDegrees=45
  twistMotion=Free
 }
 {bone1Name=dyn_earing_l, bone2Name=base_earing_l, swing1LimitDegrees=55, swing2LimitDegrees=15, twistLimitDegrees=1}
 {
  bone1Name=base_earing_r
  bone2Name=head
  swing1LimitDegrees=45
  swing1Motion=Free
  swing2LimitDegrees=45
  swing2Motion=Free
  twistLimitDegrees=45
  twistMotion=Free
 }
 {bone1Name=dyn_earing_r, bone2Name=base_earing_r, swing1LimitDegrees=55, swing2LimitDegrees=15, twistLimitDegrees=1}
 {
  bone1Name=base_hair_back
  bone2Name=head
  swing1LimitDegrees=45
  swing1Motion=Free
  swing2LimitDegrees=45
  swing2Motion=Free
  twistLimitDegrees=45
  twistMotion=Free
 }
 {
  bone1Name=base_hair_back_l
  bone2Name=base_hair_back
  swing1LimitDegrees=45
  swing1Motion=Free
  swing2LimitDegrees=45
  swing2Motion=Free
  twistLimitDegrees=45
  twistMotion=Free
 }
 {bone1Name=dyn_hair_back_l_01, bone2Name=base_hair_back_l, swing1LimitDegrees=8, swing2LimitDegrees=15, twistLimitDegrees=1}
 {bone1Name=dyn_hair_back_l_02, bone2Name=dyn_hair_back_l_01, swing1LimitDegrees=15, swing2LimitDegrees=15, twistLimitDegrees=1}
 {bone1Name=dyn_hair_back_l_03, bone2Name=dyn_hair_back_l_02, swing1LimitDegrees=40, swing2LimitDegrees=15, twistLimitDegrees=1}
 {bone1Name=dyn_hair_back_l_04, bone2Name=dyn_hair_back_l_03, swing1LimitDegrees=50, swing2LimitDegrees=15, twistLimitDegrees=1}
 {bone1Name=dyn_hair_back_l_02b, bone2Name=dyn_hair_back_l_02, swing1LimitDegrees=25, swing2LimitDegrees=15, twistLimitDegrees=3}
 {bone1Name=dyn_hair_back_l_03b, bone2Name=dyn_hair_back_l_02b, swing1LimitDegrees=40, swing2LimitDegrees=15, twistLimitDegrees=3}
 {
  bone1Name=base_hair_back_m
  bone2Name=base_hair_back
  swing1LimitDegrees=45
  swing1Motion=Free
  swing2LimitDegrees=45
  swing2Motion=Free
  twistLimitDegrees=45
  twistMotion=Free
 }
 {bone1Name=dyn_hair_back_m_01, bone2Name=base_hair_back_m, swing1LimitDegrees=8, swing2LimitDegrees=15, twistLimitDegrees=1}
 {bone1Name=dyn_hair_back_m_02, bone2Name=dyn_hair_back_m_01, swing1LimitDegrees=15, swing2LimitDegrees=15, twistLimitDegrees=1}
 {bone1Name=dyn_hair_back_m_03, bone2Name=dyn_hair_back_m_02, swing1LimitDegrees=40, swing2LimitDegrees=15, twistLimitDegrees=1}
 {bone1Name=dyn_hair_back_m_04, bone2Name=dyn_hair_back_m_03, swing1LimitDegrees=50, swing2LimitDegrees=15, twistLimitDegrees=1}
 {
  bone1Name=base_hair_back_r
  bone2Name=base_hair_back
  swing1LimitDegrees=45
  swing1Motion=Free
  swing2LimitDegrees=45
  swing2Motion=Free
  twistLimitDegrees=45
  twistMotion=Free
 }
 {bone1Name=dyn_hair_back_r_01, bone2Name=base_hair_back_r, swing1LimitDegrees=8, swing2LimitDegrees=15, twistLimitDegrees=1}
 {bone1Name=dyn_hair_back_r_02, bone2Name=dyn_hair_back_r_01, swing1LimitDegrees=15, swing2LimitDegrees=15, twistLimitDegrees=1}
 {bone1Name=dyn_hair_back_r_03, bone2Name=dyn_hair_back_r_02, swing1LimitDegrees=40, swing2LimitDegrees=15, twistLimitDegrees=1}
 {bone1Name=dyn_hair_back_r_04, bone2Name=dyn_hair_back_r_03, swing1LimitDegrees=50, swing2LimitDegrees=15, twistLimitDegrees=1}
 {bone1Name=dyn_hair_back_r_02b, bone2Name=dyn_hair_back_r_02, swing1LimitDegrees=25, swing2LimitDegrees=15, twistLimitDegrees=3}
 {bone1Name=dyn_hair_back_r_03b, bone2Name=dyn_hair_back_r_02b, swing1LimitDegrees=40, swing2LimitDegrees=15, twistLimitDegrees=3}
 {
  bone1Name=base_hair_fr_01
  bone2Name=head
  swing1LimitDegrees=45
  swing1Motion=Free
  swing2LimitDegrees=45
  swing2Motion=Free
  twistLimitDegrees=45
  twistMotion=Free
 }
 {bone1Name=dyn_hair_fr_01, bone2Name=base_hair_fr_01, swing1LimitDegrees=5, swing2LimitDegrees=5, twistLimitDegrees=1}
 {bone1Name=dyn_hair_fr_02, bone2Name=dyn_hair_fr_01, swing1LimitDegrees=20, swing2LimitDegrees=20, twistLimitDegrees=1}
 {bone1Name=dyn_hair_fr_03, bone2Name=dyn_hair_fr_02, swing1LimitDegrees=35, swing2LimitDegrees=30, twistLimitDegrees=1}
 {bone1Name=dyn_hair_fr_04, bone2Name=dyn_hair_fr_03, swing1LimitDegrees=30, swing2LimitDegrees=30, twistLimitDegrees=1}
 {
  bone1Name=dyn_hair_back_l_03
  bone2Name=dyn_hair_back_m_03
  swing1LimitDegrees=45
  swing1Motion=Free
  swing2LimitDegrees=45
  swing2Motion=Free
  twistLimitDegrees=45
  twistMotion=Free
 }
 {
  bone1Name=dyn_hair_back_r_03
  bone2Name=dyn_hair_back_m_03
  swing1LimitDegrees=45
  swing1Motion=Free
  swing2LimitDegrees=45
  swing2Motion=Free
  twistLimitDegrees=45
  twistMotion=Free
 }
 {bone1Name=dyn_hair_fr_05, bone2Name=dyn_hair_fr_04, swing1LimitDegrees=30, swing2LimitDegrees=40, twistLimitDegrees=1}
]
