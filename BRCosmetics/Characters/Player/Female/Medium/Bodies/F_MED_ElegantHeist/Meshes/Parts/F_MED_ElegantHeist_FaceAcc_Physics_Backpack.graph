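# Physics asset for the ElegantHeist medium female face accessory, Backpack variant, covering the dangling earrings and a front hair strand.
# Capsule bodies on upper spine, neck, head, right clavicle and upper arm form a free-constrained body set, likely for collision.
# Each dynamic earring hangs off a head-linked base with asymmetric swing limits and minimal twist.
# The four-link hair chain hangs from a head-linked base, with swing limits widening toward the tip.
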PhysicsAsset /BRCosmetics/Characters/Player/Female/Medium/Bodies/F_MED_ElegantHeist/Meshes/Parts/F_MED_ElegantHeist_FaceAcc_Physics_Backpack
bodies {
 base_earing_l [{shapeType=Capsule, length=0.505, radius=0.505, shapeName=base_earing_l_capsule, rotation=<rot 0 -90 0>}]
 base_earing_r [{shapeType=Capsule, length=0.505, radius=0.505, shapeName=base_earing_l_capsule, rotation=<rot 0 -90 0>}]
 base_hair_fr_01 [{shapeType=Capsule, length=0.5, radius=0.5, shapeName=base_hair_fr_01_capsule}]
 clavicle_r [{shapeType=Capsule, length=6, radius=9, shapeName=clavicle_l_capsule, center=<-7.3 2 4.2>, rotation=<rot 90 0 -10>}]
 dyn_earing_l [{shapeType=Capsule, length=1, radius=1.3, shapeName=base_earing_r_capsule, center=<4 0 0>, rotation=<rot 90 0 0>}]
 dyn_earing_r [{shapeType=Capsule, length=1, radius=1.3, shapeName=dyn_earing_r_capsule, center=<4 0 0>, rotation=<rot 90 0 0>}]
 dyn_hair_fr_01 [{shapeType=Capsule, length=2, radius=2, shapeName=dyn_hair_fr_01_capsule, rotation=<rot 0 0 90>}]
 dyn_hair_fr_02 [{shapeType=Capsule, length=3, radius=3, shapeName=dyn_hair_fr_02_capsule, center=<7.5 0 0>}]
 dyn_hair_fr_03 [{shapeType=Capsule, length=3, radius=3, shapeName=dyn_hair_fr_03_capsule, center=<5 -1.55 0.6>}]
 dyn_hair_fr_04 [{shapeType=Capsule, length=4, radius=2, shapeName=dyn_hair_fr_04_capsule, center=<1.5 -0.95 1>, rotation=<rot 45 180 0>}]
 head [{shapeType=Capsule, length=11, radius=6.2, shapeName=head_capsule, center=<3.7 2 0>, rotation=<rot 0 60 90>}]
 neck_01 [{shapeType=Capsule, length=2, radius=4.5, shapeName=neck_01_capsule, rotation=<rot 90 0 0>}]
 neck_02 [{shapeType=Capsule, length=2, radius=4.5, shapeName=neck_02_capsule, rotation=<rot 90 0 0>}]
 spine_05 [{shapeType=Capsule, length=11, radius=11, shapeName=spine_05_capsule, center=<1.15 -1 0>}]
 upperarm_r [{shapeType=Capsule, length=18, radius=8.5, shapeName=upperarm_l_capsule, center=<-16 0 -0.8>, rotation=<rot 90 0 0>}]
}
constraints [
 {
  bone1Name=clavicle_r
  bone2Name=spine_05
  swing1LimitDegrees=45
  swing1Motion=Free
  swing2LimitDegrees=45
  swing2Motion=Free
  twistLimitDegrees=45
  twistMotion=Free
 }
 {
  bone1Name=upperarm_r
  bone2Name=clavicle_r
  swing1LimitDegrees=45
  swing1Motion=Free
  swing2LimitDegrees=45
  swing2Motion=Free
  twistLimitDegrees=45
  twistMotion=Free
 }
 {
  bone1Name=neck_01
  bone2Name=spine_05
  swing1LimitDegrees=45
  swing1Motion=Free
  swing2LimitDegrees=45
  swing2Motion=Free
  twistLimitDegrees=45
  twistMotion=Free
 }
 {
  bone1Name=neck_02
  bone2Name=neck_01
  swing1LimitDegrees=45
  swing1Motion=Free
  swing2LimitDegrees=45
  swing2Motion=Free
  twistLimitDegrees=45
  twistMotion=Free
 }
 {
  bone1Name=head
  bone2Name=neck_02
  swing1LimitDegrees=45
  swing1Motion=Free
  swing2LimitDegrees=45
  swing2Motion=Free
  twistLimitDegrees=45
  twistMotion=Free
 }
 {
  bone1Name=base_earing_l
  bone2Name=head
  swing1LimitDegrees=45
  swing1Motion=Free
  swing2LimitDegrees=45
  swing2Motion=Free
  twistLimitDegrees=45
  twistMotion=Free
 }
 {bone1Name=dyn_earing_l, bone2Name=base_earing_l, swing1LimitDegrees=55, swing2LimitDegrees=15, twistLimitDegrees=1}
 {
  bone1Name=base_earing_r
  bone2Name=head
  swing1LimitDegrees=45
  swing1Motion=Free
  swing2LimitDegrees=45
  swing2Motion=Free
  twistLimitDegrees=45
  twistMotion=Free
 }
 {bone1Name=dyn_earing_r, bone2Name=base_earing_r, swing1LimitDegrees=55, swing2LimitDegrees=15, twistLimitDegrees=1}
 {
  bone1Name=base_hair_fr_01
  bone2Name=head
  swing1LimitDegrees=45
  swing1Motion=Free
  swing2LimitDegrees=45
  swing2Motion=Free
  twistLimitDegrees=45
  twistMotion=Free
 }
 {bone1Name=dyn_hair_fr_01, bone2Name=base_hair_fr_01, swing1LimitDegrees=5, swing2LimitDegrees=5, twistLimitDegrees=1}
 {bone1Name=dyn_hair_fr_02, bone2Name=dyn_hair_fr_01, swing1LimitDegrees=10, swing2LimitDegrees=10, twistLimitDegrees=1}
 {bone1Name=dyn_hair_fr_03, bone2Name=dyn_hair_fr_02, swing1LimitDegrees=15, swing2LimitDegrees=15, twistLimitDegrees=1}
 {bone1Name=dyn_hair_fr_04, bone2Name=dyn_hair_fr_03, swing1LimitDegrees=15, swing2LimitDegrees=15, twistLimitDegrees=1}
]
