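# Anim Blueprint for the ElegantHeist face accessory, driving secondary motion on the character's hair.
# It copies the parent mesh pose, constrains back-hair aim targets and runs a Control Rig, then two RigidBody simulations.
# A backpack/cape flag selects head-based simulation or spine-based with corrective constraints and bone transforms, before AnimDynamics moves front strands.
# The event graph updates gravity and sim-space settings per frame and detects back bling on customization load.
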AnimBlueprint /BRCosmetics/Characters/Player/Female/Medium/Bodies/F_MED_ElegantHeist/Meshes/Parts/F_MED_ElegantHeist_FaceAcc_AnimBP
target_skeleton=&/F_MED_ElegantHeist_FaceAcc_Skeleton
graphs [
 AnimGraph {nodes [
   n1 = Use cached pose 'CACHE_No_Backpack'()
   n2 = Use cached pose 'CACHE_With_Backpack'()
   n3 = Get IsWearingBackpackOrCape() -> (IsWearingBackpackOrCape:bool)
   n4 = Anim Dynamics - Bone: dyn_hair_fr_01!(
    SimulationSpace:AnimPhysSimSpaceType Enum
    ComponentPose=Local To Component(LocalPose=Blend Poses by bool(BlendPose_0=@n2, BlendPose_1=@n1, bActiveValue=@n3) {blendType=Linear})
    bAlphaBoolEnabled:bool=True
    ExternalForce="0,0,0"
   ) -> (Pose) {
    alphaInputType=Bool
    angularDampingOverride=1
    angularSpringConstant=10000
    bAngularSpring=true
    bOverrideAngularDamping=true
    bUseGravityOverride=true
    gravityScale=0.35
    simulationSpace=RootRelative
    alphaBoolBlend {blendInTime=0.5, blendOption=Sinusoidal, blendOutTime=1}
    boundBone {boneName=dyn_hair_fr_01}
    componentAppliedLinearAccClamp=<3000 3000 3000>
    componentLinearAccScale=<0.1 0.1 0.1>
    componentLinearVelScale=<0.2 0.2 0.2>
    gravityOverride {z=-980}
    physicsBodyDefinitions [{
      boundBone {boneName=dyn_hair_fr_01}
      boxExtents=<20 30 30>
      constraintSetup {angularLimitsMax=<10 10 10>, angularLimitsMin=<-10 -10 -30>, angularTarget=<1 0 0>, linearAxesMax=<1 1 0.5>, linearAxesMin=<-1 -1.5 -0.5>}
      localJointOffset=<4 0 0>
     }]
   }
   n5 = Anim Dynamics - Bone: dyn_hair_fr_05!(SimulationSpace:AnimPhysSimSpaceType Enum, ComponentPose=@n4, bAlphaBoolEnabled:bool=True, ExternalForce="0,0,0") -> (Pose) {
    alphaInputType=Bool
    angularDampingOverride=1
    angularSpringConstant=10000
    bAngularSpring=true
    bOverrideAngularDamping=true
    bUseGravityOverride=true
    gravityScale=0.35
    simulationSpace=RootRelative
    alphaBoolBlend {blendInTime=0.5, blendOption=Sinusoidal, blendOutTime=1}
    boundBone {boneName=dyn_hair_fr_05}
    componentAppliedLinearAccClamp=<3000 3000 3000>
    componentLinearAccScale=<0.1 0.1 0.1>
    componentLinearVelScale=<0.2 0.2 0.2>
    gravityOverride {z=-980}
    physicsBodyDefinitions [{
      boundBone {boneName=dyn_hair_fr_05}
      boxExtents=<20 30 30>
      constraintSetup {angularLimitsMax=<0 0 30>, angularLimitsMin=<0 0 0>, angularTarget=<1 0 0>, linearAxesMax=<0.5 1.5 0.5>, linearAxesMin=<-0.5 0 -0.5>}
      localJointOffset=<4 0 0>
     }]
   }
   Output Pose(Result=Component To Local(ComponentPose=@n5)) {layerGroup=None}
   n6 = "Constraint - dyn_hair_back_l_02_aimtarget (PPPPPPPP)"!(
    ComponentPose=Local To Component(LocalPose=Copy Pose From Mesh() {bUseAttachedParent=true})
    Alpha=1
    ConstraintWeights_0=0.4946
    ConstraintWeights_1:float=0.2405
    ConstraintWeights_2:float=0.2179
    ConstraintWeights_3:float=0.0333
    ConstraintWeights_4:float=0.0083
    ConstraintWeights_5:float=0.0026
    ConstraintWeights_6:float=0.0016
    ConstraintWeights_7:float=0.0012
   ) -> (Pose) {
    alphaBoolBlend {blendOption=Linear}
    boneToModify {boneName=dyn_hair_back_l_02_aimtarget}
    constraintSetup [
     {targetBone {boneName=neck_01}}
     {targetBone {boneName=head}}
     {targetBone {boneName=neck_02}}
     {targetBone {boneName=spine_04}}
     {targetBone {boneName=clavicle_l}}
     {targetBone {boneName=FACIAL_C_Jawline}}
     {targetBone {boneName=clavicle_r}}
     {targetBone {boneName=FACIAL_C_UnderChin}}
    ]
    constraintWeights [1,1,1,1,1,1,1,1]
   }
   n7 = "Constraint - dyn_hair_back_m_02_aimtarget (PPPPPPPP)"!(
    ComponentPose=@n6
    Alpha=1
    ConstraintWeights_0=0.5104
    ConstraintWeights_1:float=0.2344
    ConstraintWeights_2:float=0.2119
    ConstraintWeights_3:float=0.0312
    ConstraintWeights_4:float=0.0042
    ConstraintWeights_5:float=0.0042
    ConstraintWeights_6:float=0.0025
    ConstraintWeights_7:float=0.0012
   ) -> (Pose) {
    alphaBoolBlend {blendOption=Linear}
    boneToModify {boneName=dyn_hair_back_m_02_aimtarget}
    constraintSetup [
     {targetBone {boneName=neck_01}}
     {targetBone {boneName=head}}
     {targetBone {boneName=neck_02}}
     {targetBone {boneName=spine_04}}
     {targetBone {boneName=clavicle_r}}
     {targetBone {boneName=clavicle_l}}
     {targetBone {boneName=FACIAL_C_Jawline}}
     {targetBone {boneName=FACIAL_C_UnderChin}}
    ]
    constraintWeights [1,1,1,1,1,1,1,1]
   }
   n8 = "Constraint - dyn_hair_back_r_02_aimtarget (PPPPPPPP)"!(
    ComponentPose=@n7
    Alpha=1
    ConstraintWeights_0=0.497
    ConstraintWeights_1:float=0.2373
    ConstraintWeights_2:float=0.2173
    ConstraintWeights_3:float=0.034
    ConstraintWeights_4:float=0.0086
    ConstraintWeights_5:float=0.0027
    ConstraintWeights_6:float=0.0018
    ConstraintWeights_7:float=0.0013
   ) -> (Pose) {
    alphaBoolBlend {blendOption=Linear}
    boneToModify {boneName=dyn_hair_back_r_02_aimtarget}
    constraintSetup [
     {targetBone {boneName=neck_01}}
     {targetBone {boneName=head}}
     {targetBone {boneName=neck_02}}
     {targetBone {boneName=spine_04}}
     {targetBone {boneName=clavicle_r}}
     {targetBone {boneName=FACIAL_C_Jawline}}
     {targetBone {boneName=clavicle_l}}
     {targetBone {boneName=FACIAL_C_UnderChin}}
    ]
    constraintWeights [1,1,1,1,1,1,1,1]
   }
   n9 = Control Rig!(Source=Component To Local(ComponentPose=@n8), Alpha, IsCrouching:bool=Get bIsCrouching(), Backpack:bool=Get IsWearingBackpackOrCape()) -> (Pose) {
    alphaBoolBlend {blendOption=Linear}
    controlRigAssetReference {blueprintRigClass=&/F_MED_ElegantHeist_FaceAcc_CtrlRig.F_MED_ElegantHeist_FaceAcc_CtrlRig_C}
   }
   Save cached pose 'CACHE_Base'(Pose=@n9) {cacheName=CACHE_Base}
   n10 = RigidBody!(
    ComponentPose=Local To Component(LocalPose=Use cached pose 'CACHE_Base'())
    Alpha
    OverrideWorldGravity:Vector=Get GravityOverride()
    ExternalForce
    SimSpaceSettings:Sim Space Settings Structure=Get Sim Space Settings() -> (Sim Space Settings:Sim Space Settings Structure)
   ) -> (Pose) {
    bClampLinearTranslationLimitToRefPose=true
    bOverrideWorldGravity=true
    overridePhysicsAsset=&/F_MED_ElegantHeist_FaceAcc_Physics
    simulationSpace=BaseBoneSpace
    alphaBoolBlend {blendOption=Linear}
    baseBoneRef {boneName=head}
    componentAppliedLinearAccClamp=<3000 3000 3000>
    componentLinearAccScale=<0.4 0.3 0.3>
    componentLinearVelScale=<0.4 0.3 0.3>
    simSpaceSettings {maxAngularAcceleration=1, maxAngularVelocity=3, maxLinearAcceleration=1000, maxLinearVelocity=500, worldAlpha=1, externalLinearDragV=<1 1 1>}
   }
   Save cached pose 'CACHE_No_Backpack'(Pose=Component To Local(ComponentPose=@n10)) {cacheName=CACHE_No_Backpack}
   n11 = RigidBody(
    ComponentPose=Local To Component(LocalPose=Use cached pose 'CACHE_Base'())
    OverrideWorldGravity=Get GravityOverride()
    SimSpaceSettings=Get Sim Space Settings()
   ) {
    bClampLinearTranslationLimitToRefPose=true
    bOverrideWorldGravity=true
    overridePhysicsAsset=&/F_MED_ElegantHeist_FaceAcc_Physics_Backpack
    simulationSpace=BaseBoneSpace
    alphaBoolBlend {blendOption=Linear}
    baseBoneRef {boneName=spine_05}
    componentAppliedLinearAccClamp=<3000 3000 3000>
    componentLinearAccScale=<0.4 0.3 0.3>
    componentLinearVelScale=<0.4 0.3 0.3>
    simSpaceSettings {maxAngularAcceleration=1, maxAngularVelocity=3, maxLinearAcceleration=1000, maxLinearVelocity=500, worldAlpha=1, externalLinearDragV=<1 1 1>}
   }
   n12 = "Transform (Modify) Bone - Bone: dyn_hair_back_r_02b"(ComponentPose=@n11, Alpha=1, Translation="0,-10,0", Rotation="0,0,0", Scale="1,1,1") {
    translationMode=BMM_Additive
    translationSpace=BCS_BoneSpace
    alphaBoolBlend {blendOption=Linear}
    boneToModify {boneName=dyn_hair_back_r_02b}
    translation {y=-10}
   }
   n13 = "Constraint - dyn_hair_back_l_02b (PPPPPPPP)"!(
    ComponentPose=@n12
    Alpha=0.2
    ConstraintWeights_0=0.371
    ConstraintWeights_1:float=0.3653
    ConstraintWeights_2:float=0.1085
    ConstraintWeights_3:float=0.0664
    ConstraintWeights_4:float=0.0462
    ConstraintWeights_5:float=0.0258
    ConstraintWeights_6:float=0.0124
    ConstraintWeights_7:float=0.0044
   ) -> (Pose) {
    alphaBoolBlend {blendOption=Linear}
    boneToModify {boneName=dyn_hair_back_l_02b}
    constraintSetup [
     {targetBone {boneName=upperarm_twist_01_l}}
     {targetBone {boneName=clavicle_l}}
     {targetBone {boneName=dyn_jacket_a_01_l}}
     {targetBone {boneName=spine_05}}
     {targetBone {boneName=spine_04}}
     {targetBone {boneName=dyn_jacket_a_02_l}}
     {targetBone {boneName=deform_upperarm_bk_l}}
     {targetBone {boneName=dyn_jacket_b_01_l}}
    ]
    constraintWeights [1,1,1,1,1,1,1,1]
   }
   n14 = "Constraint - dyn_hair_back_l_02 (PPPPPPPP)"!(
    ComponentPose=@n13
    Alpha=1
    ConstraintWeights_0=0.4946
    ConstraintWeights_1:float=0.2405
    ConstraintWeights_2:float=0.2179
    ConstraintWeights_3:float=0.0333
    ConstraintWeights_4:float=0.0083
    ConstraintWeights_5:float=0.0026
    ConstraintWeights_6:float=0.0016
    ConstraintWeights_7:float=0.0012
   ) -> (Pose) {
    alphaBoolBlend {blendOption=Linear}
    boneToModify {boneName=dyn_hair_back_l_02}
    constraintSetup [
     {targetBone {boneName=neck_01}}
     {targetBone {boneName=head}}
     {targetBone {boneName=neck_02}}
     {targetBone {boneName=spine_04}}
     {targetBone {boneName=clavicle_l}}
     {targetBone {boneName=FACIAL_C_Jawline}}
     {targetBone {boneName=clavicle_r}}
     {targetBone {boneName=FACIAL_C_UnderChin}}
    ]
    constraintWeights [1,1,1,1,1,1,1,1]
   }
   n15 = "Constraint - dyn_hair_back_m_02 (PPPPPPPP)"!(
    ComponentPose=@n14
    Alpha=1
    ConstraintWeights_0=0.5104
    ConstraintWeights_1:float=0.2344
    ConstraintWeights_2:float=0.2119
    ConstraintWeights_3:float=0.0312
    ConstraintWeights_4:float=0.0042
    ConstraintWeights_5:float=0.0042
    ConstraintWeights_6:float=0.0025
    ConstraintWeights_7:float=0.0012
   ) -> (Pose) {
    alphaBoolBlend {blendOption=Linear}
    boneToModify {boneName=dyn_hair_back_m_02}
    constraintSetup [
     {targetBone {boneName=neck_01}}
     {targetBone {boneName=head}}
     {targetBone {boneName=neck_02}}
     {targetBone {boneName=spine_04}}
     {targetBone {boneName=clavicle_r}}
     {targetBone {boneName=clavicle_l}}
     {targetBone {boneName=FACIAL_C_Jawline}}
     {targetBone {boneName=FACIAL_C_UnderChin}}
    ]
    constraintWeights [1,1,1,1,1,1,1,1]
   }
   n16 = "Constraint - dyn_hair_back_r_02 (PPPPPPPP)"!(
    ComponentPose=@n15
    Alpha=1
    ConstraintWeights_0=0.497
    ConstraintWeights_1:float=0.2373
    ConstraintWeights_2:float=0.2173
    ConstraintWeights_3:float=0.034
    ConstraintWeights_4:float=0.0086
    ConstraintWeights_5:float=0.0027
    ConstraintWeights_6:float=0.0018
    ConstraintWeights_7:float=0.0013
   ) -> (Pose) {
    alphaBoolBlend {blendOption=Linear}
    boneToModify {boneName=dyn_hair_back_r_02}
    constraintSetup [
     {targetBone {boneName=neck_01}}
     {targetBone {boneName=head}}
     {targetBone {boneName=neck_02}}
     {targetBone {boneName=spine_04}}
     {targetBone {boneName=clavicle_r}}
     {targetBone {boneName=FACIAL_C_Jawline}}
     {targetBone {boneName=clavicle_l}}
     {targetBone {boneName=FACIAL_C_UnderChin}}
    ]
    constraintWeights [1,1,1,1,1,1,1,1]
   }
   n17 = "Constraint - dyn_hair_back_r_02b (PPPPPPPP)"!(
    ComponentPose=@n16
    Alpha=0.5
    ConstraintWeights_0=0.4206
    ConstraintWeights_1:float=0.2245
    ConstraintWeights_2:float=0.1854
    ConstraintWeights_3:float=0.0531
    ConstraintWeights_4:float=0.0524
    ConstraintWeights_5:float=0.0431
    ConstraintWeights_6:float=0.0116
    ConstraintWeights_7:float=0.0093
   ) -> (Pose) {
    alphaBoolBlend {blendOption=Linear}
    boneToModify {boneName=dyn_hair_back_r_02b}
    constraintSetup [
     {targetBone {boneName=upperarm_twist_01_r}}
     {targetBone {boneName=clavicle_r}}
     {targetBone {boneName=dyn_jacket_a_01_r}}
     {targetBone {boneName=spine_05}}
     {targetBone {boneName=dyn_jacket_a_02_r}}
     {targetBone {boneName=spine_04}}
     {targetBone {boneName=deform_upperarm_bk_r}}
     {targetBone {boneName=dyn_jacket_b_01_r}}
    ]
    constraintWeights [1,1,1,1,1,1,1,1]
   }
   n18 = "Constraint - dyn_hair_back_l_03b (PPPPPPPP)"!(
    ComponentPose=@n17
    Alpha=1
    ConstraintWeights_0=0.3507
    ConstraintWeights_1:float=0.2768
    ConstraintWeights_2:float=0.1376
    ConstraintWeights_3:float=0.132
    ConstraintWeights_4:float=0.0466
    ConstraintWeights_5:float=0.0378
    ConstraintWeights_6:float=0.0119
    ConstraintWeights_7:float=0.0067
   ) -> (Pose) {
    alphaBoolBlend {blendOption=Linear}
    boneToModify {boneName=dyn_hair_back_l_03b}
    constraintSetup [
     {targetBone {boneName=clavicle_l}}
     {targetBone {boneName=upperarm_twist_01_l}}
     {targetBone {boneName=spine_05}}
     {targetBone {boneName=spine_04}}
     {targetBone {boneName=dyn_jacket_a_01_l}}
     {targetBone {boneName=deform_upperarm_bk_l}}
     {targetBone {boneName=dyn_jacket_a_02_l}}
     {targetBone {boneName=spine_03}}
    ]
    constraintWeights [1,1,1,1,1,1,1,1]
   }
   n19 = "Constraint - dyn_hair_back_l_03 (PPPPPPPP)"!(
    ComponentPose=@n18
    Alpha=1
    ConstraintWeights_0=0.5296
    ConstraintWeights_1:float=0.3083
    ConstraintWeights_2:float=0.0705
    ConstraintWeights_3:float=0.0662
    ConstraintWeights_4:float=0.0153
    ConstraintWeights_5:float=0.0037
    ConstraintWeights_6:float=0.0035
    ConstraintWeights_7:float=0.0028
   ) -> (Pose) {
    alphaBoolBlend {blendOption=Linear}
    boneToModify {boneName=dyn_hair_back_l_03}
    constraintSetup [
     {targetBone {boneName=clavicle_l}}
     {targetBone {boneName=spine_05}}
     {targetBone {boneName=spine_04}}
     {targetBone {boneName=upperarm_twist_01_l}}
     {targetBone {boneName=clavicle_r}}
     {targetBone {boneName=deform_upperarm_bk_l}}
     {targetBone {boneName=dyn_jacket_a_01_l}}
     {targetBone {boneName=spine_03}}
    ]
    constraintWeights [1,1,1,1,1,1,1,1]
   }
   n20 = "Constraint - dyn_hair_back_m_03 (PPPP)"!(
    ComponentPose=@n19
    Alpha=1
    ConstraintWeights_0=0.5896
    ConstraintWeights_1:float=0.3256
    ConstraintWeights_2:float=0.0589
    ConstraintWeights_3:float=0.0259
   ) -> (Pose) {
    alphaBoolBlend {blendOption=Linear}
    boneToModify {boneName=dyn_hair_back_m_03}
    constraintSetup [{targetBone {boneName=spine_05}}, {targetBone {boneName=clavicle_r}}, {targetBone {boneName=clavicle_l}}, {targetBone {boneName=spine_04}}]
    constraintWeights [1,1,1,1]
   }
   n21 = "Constraint - dyn_hair_back_r_03 (PPPPPPPP)"!(
    ComponentPose=@n20
    Alpha=1
    ConstraintWeights_0=0.5296
    ConstraintWeights_1:float=0.3083
    ConstraintWeights_2:float=0.0705
    ConstraintWeights_3:float=0.0662
    ConstraintWeights_4:float=0.0153
    ConstraintWeights_5:float=0.0037
    ConstraintWeights_6:float=0.0035
    ConstraintWeights_7:float=0.0028
   ) -> (Pose) {
    alphaBoolBlend {blendOption=Linear}
    boneToModify {boneName=dyn_hair_back_r_03}
    constraintSetup [
     {targetBone {boneName=clavicle_r}}
     {targetBone {boneName=spine_05}}
     {targetBone {boneName=spine_04}}
     {targetBone {boneName=upperarm_twist_01_r}}
     {targetBone {boneName=clavicle_l}}
     {targetBone {boneName=deform_upperarm_bk_r}}
     {targetBone {boneName=dyn_jacket_a_01_r}}
     {targetBone {boneName=spine_03}}
    ]
    constraintWeights [1,1,1,1,1,1,1,1]
   }
   n22 = "Constraint - dyn_hair_back_r_03b (PPPPPPPP)"!(
    ComponentPose=@n21
    Alpha=1
    ConstraintWeights_0=0.382
    ConstraintWeights_1:float=0.2175
    ConstraintWeights_2:float=0.1817
    ConstraintWeights_3:float=0.0761
    ConstraintWeights_4:float=0.0601
    ConstraintWeights_5:float=0.0555
    ConstraintWeights_6:float=0.0217
    ConstraintWeights_7:float=0.0055
   ) -> (Pose) {
    alphaBoolBlend {blendOption=Linear}
    boneToModify {boneName=dyn_hair_back_r_03b}
    constraintSetup [
     {targetBone {boneName=upperarm_twist_01_r}}
     {targetBone {boneName=dyn_jacket_a_01_r}}
     {targetBone {boneName=clavicle_r}}
     {targetBone {boneName=dyn_jacket_a_02_r}}
     {targetBone {boneName=spine_05}}
     {targetBone {boneName=spine_04}}
     {targetBone {boneName=deform_upperarm_bk_r}}
     {targetBone {boneName=dyn_jacket_b_01_r}}
    ]
    constraintWeights [1,1,1,1,1,1,1,1]
   }
   n23 = "Constraint - dyn_hair_back_l_04 (PPPPPPPP)"!(
    ComponentPose=@n22
    Alpha=1
    ConstraintWeights_0=0.3756
    ConstraintWeights_1:float=0.2388
    ConstraintWeights_2:float=0.2051
    ConstraintWeights_3:float=0.126
    ConstraintWeights_4:float=0.0208
    ConstraintWeights_5:float=0.0128
    ConstraintWeights_6:float=0.0116
    ConstraintWeights_7:float=0.0093
   ) -> (Pose) {
    alphaBoolBlend {blendOption=Linear}
    boneToModify {boneName=dyn_hair_back_l_04}
    constraintSetup [
     {targetBone {boneName=clavicle_l}}
     {targetBone {boneName=spine_05}}
     {targetBone {boneName=spine_04}}
     {targetBone {boneName=upperarm_twist_01_l}}
     {targetBone {boneName=deform_upperarm_bk_l}}
     {targetBone {boneName=spine_03}}
     {targetBone {boneName=dyn_jacket_a_01_l}}
     {targetBone {boneName=clavicle_r}}
    ]
    constraintWeights [1,1,1,1,1,1,1,1]
   }
   n24 = "Constraint - dyn_hair_back_m_04 (PPPP)"!(
    ComponentPose=@n23
    Alpha=1
    ConstraintWeights_0=0.5485
    ConstraintWeights_1:float=0.2105
    ConstraintWeights_2:float=0.1214
    ConstraintWeights_3:float=0.1196
   ) -> (Pose) {
    alphaBoolBlend {blendOption=Linear}
    boneToModify {boneName=dyn_hair_back_m_04}
    constraintSetup [{targetBone {boneName=spine_05}}, {targetBone {boneName=spine_04}}, {targetBone {boneName=clavicle_r}}, {targetBone {boneName=clavicle_l}}]
    constraintWeights [1,1,1,1]
   }
   n25 = "Constraint - dyn_hair_back_r_04 (PPPPPPPP)"!(
    ComponentPose=@n24
    Alpha=1
    ConstraintWeights_0=0.3756
    ConstraintWeights_1:float=0.2388
    ConstraintWeights_2:float=0.2051
    ConstraintWeights_3:float=0.126
    ConstraintWeights_4:float=0.0208
    ConstraintWeights_5:float=0.0128
    ConstraintWeights_6:float=0.0116
    ConstraintWeights_7:float=0.0093
   ) -> (Pose) {
    alphaBoolBlend {blendOption=Linear}
    boneToModify {boneName=dyn_hair_back_r_04}
    constraintSetup [
     {targetBone {boneName=clavicle_r}}
     {targetBone {boneName=spine_05}}
     {targetBone {boneName=spine_04}}
     {targetBone {boneName=upperarm_twist_01_r}}
     {targetBone {boneName=deform_upperarm_bk_r}}
     {targetBone {boneName=spine_03}}
     {targetBone {boneName=dyn_jacket_a_01_r}}
     {targetBone {boneName=clavicle_l}}
    ]
    constraintWeights [1,1,1,1,1,1,1,1]
   }
   n26 = "Transform (Modify) Bone - Bone: dyn_hair_back_l_03b"(ComponentPose=@n25, Alpha=1, Translation="0,0,0", Rotation="0,0,0", Scale="0.8,0.6,0.8") {
    scaleMode=BMM_Additive
    scaleSpace=BCS_BoneSpace
    translationMode=BMM_Additive
    translationSpace=BCS_BoneSpace
    alphaBoolBlend {blendOption=Linear}
    boneToModify {boneName=dyn_hair_back_l_03b}
    scale=<0.8 0.6 0.8>
   }
   n27 = "Transform (Modify) Bone - Bone: dyn_hair_back_l_03"(ComponentPose=@n26, Alpha=1, Translation="0,0,0", Rotation="0,0,0", Scale="1,0.8,1") {
    scaleMode=BMM_Additive
    scaleSpace=BCS_BoneSpace
    translationMode=BMM_Additive
    translationSpace=BCS_BoneSpace
    alphaBoolBlend {blendOption=Linear}
    boneToModify {boneName=dyn_hair_back_l_03}
    scale {y=0.8}
   }
   n28 = "Transform (Modify) Bone - Bone: dyn_hair_back_m_03"(ComponentPose=@n27, Alpha=1, Translation="0,0,0", Rotation="0,0,0", Scale="1,0.8,1") {
    scaleMode=BMM_Additive
    scaleSpace=BCS_BoneSpace
    translationMode=BMM_Additive
    translationSpace=BCS_BoneSpace
    alphaBoolBlend {blendOption=Linear}
    boneToModify {boneName=dyn_hair_back_m_03}
    scale {y=0.8}
   }
   n29 = "Transform (Modify) Bone - Bone: dyn_hair_back_r_03"(ComponentPose=@n28, Alpha=1, Translation="0,0,0", Rotation="0,0,0", Scale="1,0.8,1") {
    scaleMode=BMM_Additive
    scaleSpace=BCS_BoneSpace
    translationMode=BMM_Additive
    translationSpace=BCS_BoneSpace
    alphaBoolBlend {blendOption=Linear}
    boneToModify {boneName=dyn_hair_back_r_03}
    scale {y=0.8}
   }
   n30 = "Transform (Modify) Bone - Bone: dyn_hair_back_r_03b"(ComponentPose=@n29, Alpha=1, Translation="0,0,0", Rotation="0,0,0", Scale="0.8,0.6,0.8") {
    scaleMode=BMM_Additive
    scaleSpace=BCS_BoneSpace
    translationMode=BMM_Additive
    translationSpace=BCS_BoneSpace
    alphaBoolBlend {blendOption=Linear}
    boneToModify {boneName=dyn_hair_back_r_03b}
    scale=<0.8 0.6 0.8>
   }
   Save cached pose 'CACHE_With_Backpack'(Pose=Component To Local(ComponentPose=@n30)) {cacheName=CACHE_With_Backpack}
  ]}
 EventGraph {nodes [
   Event BlueprintUpdateAnimation()
   > n1 = Sequence()
   > n2 = Gravity Override State Function DEPRECATED(
    CharacterPartAnimInstance=Self-Reference()
    GameTimeInSeconds=GetGameTimeInSeconds()
    GravityOverrideParams=Get Gravity Override Params()
    RBAN Params=Get RBAN Params()
   )
   > Set GravityOverride(GravityOverride=@n2)
   > Set Sim Space Settings(Sim Space Settings:Sim Space Settings Structure=Break RBANSetting Struct!(RBANSettingStruct:"RBANSetting Struct Structure (by ref)"=@n2.RBAN SettingsOut) -> (
     ComponentLinearAccScale_28_#1:Vector=~
     ComponentLinearVelScale_31_#2:Vector=~
     SimSpaceSettings_34_#3:Sim Space Settings Structure
    ) {is=title K2Node_BreakStruct}.SimSpaceSettings_34_#3) -> (Output_Get:Sim Space Settings Structure)
   n3 = Branch(execute=@n1.then_1, Condition=Get bIsInFrontEnd())
   > Set AnimDynamic_SimulationSpace(AnimDynamic_SimulationSpace:Byte=2) -> (Output_Get:Byte=0)
   Set AnimDynamic_SimulationSpace(execute=@n3.else, AnimDynamic_SimulationSpace=3) -> (Output_Get=0)
   Event OnCharacterCustomizationLoaded()
   > n4 = Cast To PlayerPawn_Athena(Object=TryGetPawnOwner())
   > Set Body_Mesh(Body_Mesh=GetSkeletalMeshForPartType(self=@n4))
   n5 = CheckForBackbling()
   > Set IsWearingBackpackOrCape(IsWearingBackpackOrCape:bool=OR Boolean!(A=@n5, B=@n5.bIsBackblingACape, C:bool=@n5.bIsBackblingLong) -> (ReturnValue)) -> (Output_Get:bool)
   > Set isWearingBackpack(isWearingBackpack:bool=AND Boolean!(A=@n5, B=NOT Boolean(A=@n5.bIsBackblingACape), C:bool=NOT Boolean(A=@n5.bIsBackblingLong)) -> (ReturnValue)) -> (Output_Get:bool)
   > Set Is Wearing Cape(Is Wearing Cape=OR Boolean!(A=@n5.bIsBackblingACape, B=@n5.bIsBackblingLong) -> (ReturnValue))
  ]}
]
variables [
 AnimDynamic_SimulationSpace {type=integer}
 Body_Mesh {type=SkeletalMeshComponent}
 Gravity Override Params {type=GravityOverrideParamsStruct}
 GravityOverride {type=Vector}
 Is Wearing Cape {type=boolean}
 isWearingBackpack {type=boolean}
 IsWearingBackpackOrCape {type=boolean}
 RBAN Params {
  type=RBANParamsStruct
  default {
   dBNO {simSpaceSettings {
     maxAngularAcceleration=1000
     maxAngularVelocity=3
     maxLinearAcceleration=300
     maxLinearVelocity=700
     velocityScaleZ=0
     externalLinearDragV=<0 0.5 0>
    }}
   dBNO_Carried {simSpaceSettings {
     maxAngularAcceleration=1000
     maxAngularVelocity=3
     maxLinearAcceleration=300
     maxLinearVelocity=700
     velocityScaleZ=0
     externalLinearDragV=<0 0.5 0>
    }}
   emoteOrMelee {simSpaceSettings {
     maxAngularAcceleration=1000
     maxAngularVelocity=3
     maxLinearAcceleration=300
     maxLinearVelocity=700
     velocityScaleZ=1
     externalLinearDragV=<0 0.5 0>
    }}
   front_End {simSpaceSettings {
     maxAngularAcceleration=1000
     maxAngularVelocity=8
     maxLinearAcceleration=500
     maxLinearVelocity=700
     velocityScaleZ=1
     externalLinearDragV=<0.5 1 0.5>
    }}
   onGround_Crouch_Moving {simSpaceSettings {
     maxAngularAcceleration=1000
     maxAngularVelocity=3
     maxLinearAcceleration=300
     maxLinearVelocity=700
     velocityScaleZ=0
     externalLinearDragV=<0 0.5 0>
    }}
   onGround_Falling {simSpaceSettings {
     maxAngularAcceleration=1000
     maxAngularVelocity=3
     maxLinearAcceleration=300
     maxLinearVelocity=700
     velocityScaleZ=0
     externalLinearDragV=<0 0.5 0>
    }}
   onGround_Moving {simSpaceSettings {
     maxAngularAcceleration=1000
     maxAngularVelocity=3
     maxLinearAcceleration=300
     maxLinearVelocity=700
     velocityScaleZ=0
     externalLinearDragV=<0 0.5 0>
    }}
   onGround_Standing {simSpaceSettings {
     maxAngularAcceleration=1000
     maxAngularVelocity=3
     maxLinearAcceleration=300
     maxLinearVelocity=700
     velocityScaleZ=0
     externalLinearDragV=<0 0.5 0>
    }}
   rail_Grinding {simSpaceSettings {
     maxAngularAcceleration=1000
     maxAngularVelocity=3
     maxLinearAcceleration=300
     maxLinearVelocity=700
     velocityScaleZ=0
     worldAlpha=0.1
     externalLinearDragV=<0 0.5 0>
    }}
   riding_Moving {simSpaceSettings {
     maxAngularAcceleration=1000
     maxAngularVelocity=3
     maxLinearAcceleration=300
     maxLinearVelocity=700
     velocityScaleZ=0
     externalLinearDragV=<0 0.5 0>
    }}
   riding_Moving_Targeting {simSpaceSettings {
     maxAngularAcceleration=1000
     maxAngularVelocity=3
     maxLinearAcceleration=300
     maxLinearVelocity=700
     velocityScaleZ=0
     externalLinearDragV=<0 0.5 0>
    }}
   riding_Standing {simSpaceSettings {
     maxAngularAcceleration=1000
     maxAngularVelocity=3
     maxLinearAcceleration=300
     maxLinearVelocity=700
     velocityScaleZ=0
     externalLinearDragV=<0 0.5 0>
    }}
   skydive_DivingDown {simSpaceSettings {
     maxAngularAcceleration=1000
     maxAngularVelocity=3
     maxLinearAcceleration=300
     maxLinearVelocity=700
     velocityScaleZ=0
     externalLinearDragV=<0 0.5 0>
    }}
   skydive_FloatingDown {
    componentLinearAccScale=<0 0.5 0>
    componentLinearVelScale=<0 0.5 0>
    simSpaceSettings {
     maxAngularAcceleration=1000
     maxAngularVelocity=3
     maxLinearAcceleration=300
     maxLinearVelocity=700
     velocityScaleZ=0
     externalLinearDragV=<0 0.5 0>
    }
   }
   skydive_GoingUp {simSpaceSettings {
     maxAngularAcceleration=1000
     maxAngularVelocity=3
     maxLinearAcceleration=300
     maxLinearVelocity=700
     velocityScaleZ=0
     externalLinearDragV=<0 0.5 0>
    }}
   skydive_Parachuting {
    componentLinearAccScale=<0 0.1 0.3>
    componentLinearVelScale=<0 0.1 0.3>
    simSpaceSettings {
     maxAngularAcceleration=1000
     maxAngularVelocity=3
     maxLinearAcceleration=300
     maxLinearVelocity=700
     velocityScaleZ=0
     externalLinearDragV=<0 0.5 0>
    }
   }
   skydive_Surfing {simSpaceSettings {
     maxAngularAcceleration=1000
     maxAngularVelocity=3
     maxLinearAcceleration=300
     maxLinearVelocity=700
     velocityScaleZ=0
     externalLinearDragV=<0 0.5 0>
    }}
   swimming {simSpaceSettings {
     maxAngularAcceleration=1000
     maxAngularVelocity=3
     maxLinearAcceleration=300
     maxLinearVelocity=700
     velocityScaleZ=0
     externalLinearDragV=<0 0.5 0>
    }}
   tactical_Sprint {simSpaceSettings {
     maxAngularAcceleration=1000
     maxAngularVelocity=3
     maxLinearAcceleration=300
     maxLinearVelocity=700
     velocityScaleZ=0
     externalLinearDragV=<0 0.5 0>
    }}
   ziplining {simSpaceSettings {
     maxAngularAcceleration=1000
     maxAngularVelocity=3
     maxLinearAcceleration=300
     maxLinearVelocity=700
     velocityScaleZ=0
     externalLinearDragV=<0 0.5 0>
    }}
  }
 }
 Sim Space Settings {
  type=SimSpaceSettings
  default {
   dampingAlpha=1
   maxAngularAcceleration=1000
   maxAngularVelocity=3
   maxLinearAcceleration=300
   maxLinearVelocity=700
   velocityScaleZ=1
   worldAlpha=1
   externalLinearDragV=<0.5 1 0.5>
  }
 }
]
